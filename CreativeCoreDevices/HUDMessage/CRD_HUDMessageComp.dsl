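; Per-player component for the HUD Message device that decides which device's message the owning player controller sees, and cleans itself up on minigame end or player removal.
; The server keeps priority-bucketed queues with timeouts, display timers and relevance rechecks, batching client updates to the next frame.
; The client waits for outros, formats text, plays audio and drives a Dynamic UI scene through gameplay events, resending once the widget is ready.
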(event Custom|Player%20Changed%20Team
  (params (PlayerState :type Fort%20Player%20State%20Object%20Reference) (TeamId :type Byte) (OldTeamId :type Byte) (bTeamChangedWithoutRespawn :type Boolean))
  (bind _returnvalue (Fortnite|GetFortPlayerControllerfromActor PlayerState))
  ;[0 0.6589 1 0.26] When the owning player changes teams, check to see if we're still relevant (if we're set to). If not, remove current message.
  (bind _controlling_device (Variables|Default|GetControlling%20Device)
    (:then
      (if (== _returnvalue (Components|GetOwner))
        (if (Variables|/CreativeCoreDevices/Device_HUDMessage_V2.Device_HUDMessage_V2_C|GetRe-Evaluate%20Messages%20on%20Show _controlling_device)
          (bind _true (Class|DeviceHUDMessageV2|IsRelevantForPlayer _controlling_device _returnvalue true (Variables|/CreativeCoreDevices/Device_HUDMessage_V2.Device_HUDMessage_V2_C|GetLastInstigator _controlling_device)))
          (if (not _true)
            (Server|UpdateControllingDevicefromQueue)))))
    (:else)))

(event Custom|Clear%20All%20Messages%20
  ;[0 0.5408 1 0.2] Controlling Device Updates - These update what device is controlling this component (e.g. what Device is currently showing its Message)
  (Utilities|Array|Clear (Variables|Default|GetQueued%20Devices))
  (Server|UpdateControllingDevice 0))

(event Custom|Update%20Client%20Next%20Frame
  (params (New%20Controlling%20Device%20GUID :type Guid%20Structure) (Message%20Details :type Gameplay%20Event%20HUDMessage%20Details%20Structure) (Instigator :type Player%20State%20Object%20Reference))
  ;[0 0.5408 1 0.2] Controlling Device Updates - These update what device is controlling this component (e.g. what Device is currently showing its Message)
  (Utilities|FlowControl|RetriggerableDelay 0.0)
  (CallFunction|RPCCLIENTShowMessage New%20Controlling%20Device%20GUID Message%20Details Instigator)
  (CallFunction|StartTimerForControlSwitch))

(event Custom|AddMessage
  (params (DeviceRef :type Device%20HUDMessage%20V2%20Object%20Reference) (MessagePayload :type HUDMessage%20Queue%20Details%20Structure))
  ;[0 0.5408 1 0.2] Controlling Device Updates - These update what device is controlling this component (e.g. what Device is currently showing its Message)
  (Utilities|TestPriority DeviceRef
    (:Should%20Take%20Over
      (Server|UpdateControllingDevice DeviceRef MessagePayload))
    (:Don't%20Take%20Over
      (Server|AddDeviceToQueue DeviceRef MessagePayload))))

(event Custom|Start%20Timer%20for%20Control%20Switch
  ;[0 0.5408 1 0.2] Controlling Device Updates - These update what device is controlling this component (e.g. what Device is currently showing its Message)
  (Variables|Default|GetControlling%20Device
    (:then
      (bind _time (Utilities|ServerGetDisplayTime)
        (:HasTime
          (Utilities|Time|SetTimerbyEvent (delegate Custom|On%20Controlling%20Device%20Display%20Time%20Complete) _time))))
    (:else)))

(event EventBeginPlay
  (bind _returnvalue (Components|GetOwner))
  (bind _returnvalue_2 (Fortnite|GetFortPlayerControllerfromActor _returnvalue))
  ;[0 1 0.2152 0.2] Initialization - Called when the first device attempts to create this component. Bind to the minigame ended delegate.
  (bind _returnvalue_3 (Minigame|GetMinigameforActor (Components|GetOwner)))
  (Utilities|IsValid _returnvalue_3
    (:Is%20Valid
      (Minigame|BindEventtoMinigameEndedDelegate _returnvalue_3 (delegate Custom|On%20Minigame%20Ended))
      (Minigame|BindEventtoPlayerPreRemovedDelegate _returnvalue_3 (delegate Custom|Player%20Removed))
      (Minigame|BindEventtoPlayerChangedTeamDelegate _returnvalue_3 (delegate Custom|Player%20Changed%20Team))
      (Utilities|IsValid _returnvalue_2
        (:Is%20Valid
          (Variables|Default|SetClientCachedController _returnvalue_2))))))

(event Custom|On%20Controlling%20Device%20Display%20Time%20Complete
  ;[0 0.5408 1 0.2] Controlling Device Updates - These update what device is controlling this component (e.g. what Device is currently showing its Message)
  (Server|UpdateControllingDevicefromQueue))

(event Custom|RemoveDevice
  (params (DeviceRef :type Device%20HUDMessage%20V2%20Object%20Reference))
  ;[0 0.5408 1 0.2] Controlling Device Updates - These update what device is controlling this component (e.g. what Device is currently showing its Message)
  (Server|FindandRemoveDevice DeviceRef)
  ;;[0 0.3391 1 0.1] If we remove the current displayed object, stop being the controlling device and then update (else it could re-add itself to the queue)
  (if (Utilities|Equal(Object) DeviceRef (Variables|Default|GetControlling%20Device))
    (Variables|Default|SetControlling%20Device 0)
    (Variables|Default|SetServer%20Cached%20Details)
    (bind _queue_empty (Server|UpdateControllingDevicefromQueue))
    (if _queue_empty
      (CallFunction|UpdateClientNextFrame) ; Manually calling hide here (because the update function won't do it as it thinks we're already empty)
    )))

(event Custom|On%20Minigame%20Ended
  ;[1 0.0129 0 0.2] Destruction - On Game end or owning player removed from Minigame. clean up this component (we don't need hud messages during edit, and this ensures everything is clean for next session)
  (Components|DestroyComponent))

(event EventEndPlay
  (params (EndPlayReason :type EEndPlayReason%20Enum :default "Destroyed"))
  ;[1 0.0129 0 0.2] Destruction - On Game end or owning player removed from Minigame. clean up this component (we don't need hud messages during edit, and this ensures everything is clean for next session)
  (Client|ClientClearWidget false)
  (bind _clientcachedcontroller (Variables|Default|GetClientCachedController)
    (:then
      (DynamicUI|RemoveScene (WorldSubsystems|GetDynamicUIManager) (Variables|Default|GetDynamicHUDScene) _clientcachedcontroller)
      (Variables|Default|SetClientCachedController 0))
    (:else)))

(event Custom|Player%20Removed
  (params (UniqueNetId :type Unique%20Net%20Id%20Structure) (bIsLocalPlayer :type Boolean))
  (bind _returnvalue (Game|GetPlayerStatefromUniqueNetId UniqueNetId))
  ;[1 0.0129 0 0.2] Destruction - On Game end or owning player removed from Minigame. clean up this component (we don't need hud messages during edit, and this ensures everything is clean for next session)
  (Utilities|IsValid _returnvalue
    (:Is%20Valid
      (if (== (Fortnite|GetFortPlayerControllerfromActor _returnvalue) (Components|GetOwner))
        (Components|DestroyComponent)))))

(event Custom|Client%20Send%20Widget%20Details%20after%20initialization
  (params (Details :type Gameplay%20Event%20HUDMessage%20Details%20Structure))
  ;[1 0 0.3002 0] If we've failed to send our initialization message (because there's no listening object yet). This will send again when there is.
  (Utilities|FlowControl|DoOnce false
    (:Completed
      (bind _event_router (Variables|Default|GetEvent%20Router)
        (:then
          (bind _clientcachedcontroller (Variables|Default|GetClientCachedController)
            (:then
              (bind _asynctaskproxy (GameplayEventRouter|StartListeningToEvent _event_router "/CreativeCoreDevices/HUDMessage/GameplayEvent_HUDMessageReady.GameplayEvent_HUDMessageReady" _clientcachedcontroller)
                (:OnEventReceived
                  (GameplayEventRouter|StopListeningtoEvent _asynctaskproxy)
                  (GameplayEventRouter|BroadcastEvent (Variables|Default|GetEvent%20Router) Details _clientcachedcontroller))))
            (:else)))
        (:else)))))

(event Custom|On%20Widget%20Outro%20Complete
  ;[1 0.4884 0 0.2] CLIENT LOGIC - On Replication of a new controlling device, update the widget, optionally with a short delay in order to allow the existing UI to hide.
  (if (Guid|IsValidGuid (Variables|Default|GetClient%20Control%20GUID))
    (CallFunction|ClientCreateAndUpdateMessageWidget)))

(event Custom|RPC_CLIENT%20Show%20Message
  (params (New%20Controlling%20Device%20GUID :type Guid%20Structure) (Message%20Details :type Gameplay%20Event%20HUDMessage%20Details%20Structure) (Instigator :type Player%20State%20Object%20Reference))
  ;[1 0.4884 0 0.2] CLIENT LOGIC - On Replication of a new controlling device, update the widget, optionally with a short delay in order to allow the existing UI to hide.
  (if (Guid|IsValidGuid New%20Controlling%20Device%20GUID)
    (if (Guid|IsValidGuid (Variables|Default|GetClient%20Control%20GUID))
      ;;[0 0.3128 1 0.2] If we have an existing widget, tell it to clear, then wait till it should be finished and update it.
      (bind _hasclosinganim (Client|ClientClearWidget false))
      (Client|ClientCacheMessageDetails New%20Controlling%20Device%20GUID Message%20Details Instigator)
      (if _hasclosinganim
        (bind _returnvalue (Utilities|Time|SetTimerbyEvent (delegate Custom|On%20Widget%20Outro%20Complete) (Variables|Default|GetWidgetOutroTime)))
        (Variables|Default|SetWidget%20Outro%20Timer _returnvalue)
        (elif (Guid|IsValidGuid (Variables|Default|GetClient%20Control%20GUID))
          _))
      (else
        (Utilities|Time|ClearandInvalidateTimerbyHandle (Variables|Default|GetWidget%20Outro%20Timer))
        (Client|ClientCacheMessageDetails New%20Controlling%20Device%20GUID Message%20Details Instigator)))
    (CallFunction|ClientCreateAndUpdateMessageWidget)
    (else
      (Client|ClientClearWidget false))))

(event Custom|Client%20Create%20and%20Update%20Message%20Widget
  (bind _returnvalue (WorldSubsystems|GetDynamicUIManager))
  (bind _dynamichudscene (Variables|Default|GetDynamicHUDScene))
  ;[1 0.635 0 0.2] Create the HUD Message Widget when necessary.
  (Networking|CanExecuteCosmeticEvents
    (:True
      (bind _clientcachedcontroller (Variables|Default|GetClientCachedController)
        (:then
          (if (not (DynamicUI|IsSceneActive _returnvalue _dynamichudscene _clientcachedcontroller))
            (DynamicUI|AddScene _returnvalue _dynamichudscene _clientcachedcontroller))
          (Client|ClientUpdateWidgetFromDetails))
        (:else)))))

(event Custom|Load%20and%20Play%20Sound
  (bind _client_cached_details (Variables|Default|GetClient%20Cached%20Details))
  (bind (_message_72_7634162b4c068206f407bcac0c836150 _textjustification_75_68386cab41ed610c8696fa80e715b116 _audio_31_f490e2304d7f53d4d13035a1db3f05f9) (Utilities|Struct|BreakGameplayEventHUDMessageDetails _client_cached_details))
  ;[1 0.6509 0 0.25] Load and Play 2d Audio from the device when the message displays.
  (if (Utilities|IsValidSoftObjectReference _audio_31_f490e2304d7f53d4d13035a1db3f05f9)
    (bind _object (Utilities|AsyncLoadAsset _audio_31_f490e2304d7f53d4d13035a1db3f05f9)
      (:Completed
        (bind _assound_base (Utilities|Casting|CastToSoundBase _object))
        (Audio|PlaySound2D _assound_base)))))

(fn AddDevicetoQueue
  (params (Device :type Device%20HUDMessage%20V2%20Object%20Reference) (MessagePayload :type HUDMessage%20Queue%20Details%20Structure))
  (bind _priority (LocalVariables|GetPriority))
  (bind _queued_devices (Variables|Default|GetQueued%20Devices))
  (bind (_output_device_4_1d5858c6423b2ef1df5985976caeba50 _output_message_11_5f74f8bd407bc035e95aaf8e0557f0a0) (Utilities|Struct|BreakHUDMessageQueue (Utilities|Array|Get(acopy) _queued_devices _priority)))
  (bind (_output_device_4_1d5858c6423b2ef1df5985976caeba50_2 _output_message_11_5f74f8bd407bc035e95aaf8e0557f0a0_2) (Utilities|Struct|BreakHUDMessageQueue (Utilities|Array|Get(acopy) (Variables|Default|GetQueued%20Devices) (LocalVariables|GetPriority))))
  (Utilities|IsValid Device
    (:Is%20Valid
      (if (> (Variables|/CreativeCoreDevices/Device_HUDMessage_V2.Device_HUDMessage_V2_C|GetQueueTimeout Device) 0)
        (if (or (Variables|/CreativeCoreDevices/Device_HUDMessage_V2.Device_HUDMessage_V2_C|GetAllow%20Multiple%20in%20Queue (Variables|GetDevice)) (!= (Variables|Default|GetControlling%20Device) (Variables|GetDevice)))
          (bind _output_get (LocalVariables|SetPriority (Math|Integer|Clamp(Integer) (Variables|/CreativeCoreDevices/Device_HUDMessage_V2.Device_HUDMessage_V2_C|GetPriority (Variables|GetDevice)) 0 10)))
          (if (Utilities|Array|IsValidIndex (Variables|Default|GetQueued%20Devices) _output_get)
            (LocalVariables|SetLocalArray _output_device_4_1d5858c6423b2ef1df5985976caeba50)
            (LocalVariables|SetLocalDetailsArray _output_message_11_5f74f8bd407bc035e95aaf8e0557f0a0)
            (if (not (Variables|/CreativeCoreDevices/Device_HUDMessage_V2.Device_HUDMessage_V2_C|GetAllow%20Multiple%20in%20Queue (Variables|GetDevice)))
              (bind _output_get_2 (LocalVariables|SetDeviceIndex (Utilities|Array|FindItem (LocalVariables|GetLocalArray) (Variables|GetDevice))))
              (if (== _output_get_2 -1)
                (LocalVariables|SetDeviceIndex (Utilities|Array|Length (LocalVariables|GetLocalArray)))
                (elif (CallFunction|MessageDiffers (Utilities|Array|Get(acopy) _output_message_11_5f74f8bd407bc035e95aaf8e0557f0a0_2 _output_get_2) (Variables|GetMessagePayload))
                  _
                  (else
                    (return false))))
              (if (< (LocalVariables|GetDeviceIndex) (Variables|Default|GetMaxArrayQueue))
                (Utilities|Array|SetArrayElem (LocalVariables|GetLocalArray) (LocalVariables|GetDeviceIndex) (Variables|GetDevice) true)
                (Utilities|Array|SetArrayElem (LocalVariables|GetLocalDetailsArray) (LocalVariables|GetDeviceIndex) (Variables|GetMessagePayload) true)
                (Utilities|Array|SetArrayElem (Variables|Default|GetQueued%20Devices) (LocalVariables|GetPriority) (Utilities|Struct|MakeHUDMessageQueue (LocalVariables|GetLocalArray) (LocalVariables|GetLocalDetailsArray)) false)
                (return true)
                (else
                  _)))
            (else
              (Utilities|Array|SetArrayElem (Variables|Default|GetQueued%20Devices) (LocalVariables|GetPriority) (Utilities|Struct|MakeHUDMessageQueue (Utilities|Array|MakeArray (Variables|GetDevice)) (Utilities|Array|MakeArray (Variables|GetMessagePayload))) true)
              (return true))))))))

(fn ClientCacheMessageDetails
  (params (Client%20control%20GUID :type Guid%20Structure) (Client%20Cached%20Details :type Gameplay%20Event%20HUDMessage%20Details%20Structure) (Instigator :type Player%20State%20Object%20Reference))
  (bind (_message_72_7634162b4c068206f407bcac0c836150 _textjustification_75_68386cab41ed610c8696fa80e715b116 _styleoverride_78_c05eca1e4be85cc124e69e8824f56f39) (Utilities|Struct|BreakGameplayEventHUDMessageDetails Client%20Cached%20Details))
  (bind _name (Client|GetInstigatorName Instigator))
  (bind _value (Class|LocalizablePayloadFunctionLibrary|LocalizablePayloadtoText _message_72_7634162b4c068206f407bcac0c836150))
  (bind _clientcachedcontroller (Variables|Default|GetClientCachedController)
    (:then
      (Variables|Default|SetClient%20Control%20GUID Client%20control%20GUID)
      (Variables|Default|SetClient%20Cached%20Details Client%20Cached%20Details)
      (bind _output_get (LocalVariables|SetStyleOverride _styleoverride_78_c05eca1e4be85cc124e69e8824f56f39))
      (if (Utilities|String|IsEmpty _output_get)
        _
        (elif (Utilities|Text|TextIsEmpty (LocalizableMessage|ConvLocalizableMessagetoText (Utilities|Struct|BreakLocalizablePayload _message_72_7634162b4c068206f407bcac0c836150)))
          _
          (else
            (Utilities|Struct|SetmembersinGameplayEventHUDMessageDetails (Variables|Default|GetClient%20Cached%20Details) (Class|LocalizablePayloadFunctionLibrary|TexttoLocalizablePayload (Utilities|Text|FormatText "NSLOCTEXT(\"[D6A725B11D9B08F0034DC6CD139F00B9]\", \"ECD93A714C585E50C2402AA3F341EDB2\", \"<{style}>{message}</>\")" (LocalVariables|GetStyleOverride) (Utilities|Text|FormatText _value (Creative|UI|GenericHelper|GetLocalPlayerNameFromController _clientcachedcontroller) _name)))) ; If we have a style override, wrap the entire message in it.\nThis allows Verse-defined messages to apply markup without having to \nmanually add it (important for backwards compat)
            (return))))
      (Utilities|Struct|SetmembersinGameplayEventHUDMessageDetails (Variables|Default|GetClient%20Cached%20Details) (Class|LocalizablePayloadFunctionLibrary|TexttoLocalizablePayload (Utilities|Text|FormatText _value (Creative|UI|GenericHelper|GetLocalPlayerNameFromController _clientcachedcontroller) _name))))
    (:else)))

(fn ClientClearWidget
  (params (Immediate :type Boolean))
  (bind (_message_72_7634162b4c068206f407bcac0c836150 _textjustification_75_68386cab41ed610c8696fa80e715b116 _outroanimation_65_7f0532384fd0076dfd602088e19f9b5f) (Utilities|Struct|BreakGameplayEventHUDMessageDetails (Variables|Default|GetClient%20Cached%20Details)))
  (LocalVariables|SetLocalOutgoingAnimation _outroanimation_65_7f0532384fd0076dfd602088e19f9b5f)
  (Utilities|Time|ClearandInvalidateTimerbyHandle (Variables|Default|GetWidget%20Outro%20Timer))
  (Guid|InvalidateGuid (Variables|Default|GetClient%20Control%20GUID))
  (Variables|Default|SetClient%20Cached%20Details)
  (bind _event_router (Variables|Default|GetEvent%20Router)
    (:then
      (bind _clientcachedcontroller (Variables|Default|GetClientCachedController)
        (:then
          (GameplayEventRouter|BroadcastEvent _event_router (Utilities|Struct|MakeGameplayEventHUDMessageClear (Variables|Default|GetHUD%20Layer) Immediate) _clientcachedcontroller))
        (:else
          _))
      (return (Utilities|Enum|NotEqual(Enum) (LocalVariables|GetLocalOutgoingAnimation) "NewEnumerator0")))
    (:else
      _)))

(fn ClientUpdateWidgetfromDetails
  (bind _client_cached_details (Variables|Default|GetClient%20Cached%20Details))
  (CallFunction|LoadAndPlaySound)
  (bind _clientcachedcontroller (Variables|Default|GetClientCachedController)
    (:then
      (bind _returnvalue (GameplayEventRouter|GetEventRouter "/Script/FortniteGame.FortPlayspace" _clientcachedcontroller))
      (bind _output_get (Variables|Default|SetEvent%20Router _returnvalue))
      (GameplayEventRouter|BroadcastEvent _output_get _client_cached_details _clientcachedcontroller)
      (CallFunction|ClientSendWidgetDetailsAfterInitialization _client_cached_details))
    (:else)))

(fn FindAndRemoveDevice
  (params (ObjectRef :type Device%20HUDMessage%20V2%20Object%20Reference))
  (bind _priority (LocalVariables|GetPriority))
  (bind _queued_devices (Variables|Default|GetQueued%20Devices))
  (bind (_output_device_4_1d5858c6423b2ef1df5985976caeba50 _output_message_11_5f74f8bd407bc035e95aaf8e0557f0a0) (Utilities|Struct|BreakHUDMessageQueue (Utilities|Array|Get(acopy) _queued_devices _priority)))
  (bind _foundindex (LocalVariables|GetFoundIndex))
  (bind _output_get (LocalVariables|SetPriority (Math|Integer|Clamp(Integer) (Variables|/CreativeCoreDevices/Device_HUDMessage_V2.Device_HUDMessage_V2_C|GetPriority ObjectRef) 0 10)))
  (if (Utilities|Array|IsValidIndex (Variables|Default|GetQueued%20Devices) _output_get)
    (bind _output_get_2 (LocalVariables|SetFoundDeviceArray _output_device_4_1d5858c6423b2ef1df5985976caeba50))
    (LocalVariables|SetDetailsArray _output_message_11_5f74f8bd407bc035e95aaf8e0557f0a0)
    (bind _output_get_3 (LocalVariables|SetFoundIndex (Utilities|Array|FindItem _output_get_2 (Variables|GetObjectRef))))
    (if (>= _output_get_3 0)
      (Utilities|Array|RemoveIndex (LocalVariables|GetFoundDeviceArray) _foundindex)
      (Utilities|Array|RemoveIndex (LocalVariables|GetDetailsArray) _foundindex)
      (Utilities|Array|SetArrayElem (Variables|Default|GetQueued%20Devices) (LocalVariables|GetPriority) (Utilities|Struct|MakeHUDMessageQueue (LocalVariables|GetFoundDeviceArray) (LocalVariables|GetDetailsArray)) false)
      (return (Utilities|Array|IsEmpty (LocalVariables|GetFoundDeviceArray)))
      (else
        (return false)))
    (else
      (return false))))

(fn GetInstigatorName
  (params (PS :type Player%20State%20Object%20Reference))
  (bind _asfort_player_state_athena (Utilities|Casting|CastToFortPlayerStateAthena PS)
    (:then
      (return (Utilities|Text|ToText(String) (PlayerState|GetPlayerNameforStreaming _asfort_player_state_athena))))
    (:CastFailed
      (return ""))))

(fn GetRemainingDisplayTime
  (bind (_server_cached_details_timeadded_2_6d40806d4734592914ecb18c9b114de8 _server_cached_details_instigatorps_13_c6228da247aa3cc4ada8b28c74dad64c _server_cached_details_messagepayload_11_b3fb8ab54557dceb753f2091d88bd21a) (Utilities|Struct|BreakHUDMessageQueueDetails (Variables|GetServer%20Cached%20Details_TimeAdded_2_6D40806D4734592914ECB18C9B114DE8)))
  (bind (_message_72_7634162b4c068206f407bcac0c836150 _textjustification_75_68386cab41ed610c8696fa80e715b116 _displaytime_37_33f9b57c45f05e666a7eefb4fd681bf5) (Utilities|Struct|BreakGameplayEventHUDMessageDetails _server_cached_details_messagepayload_11_b3fb8ab54557dceb753f2091d88bd21a))
  (Variables|Default|GetControlling%20Device
    (:then
      (return (Math|Float|SelectFloat (- (+ _displaytime_37_33f9b57c45f05e666a7eefb4fd681bf5 (Variables|Default|GetLastDisplayedTime)) (Utilities|Time|GetGameTimeinSeconds)) 99999999999999.0 (> _displaytime_37_33f9b57c45f05e666a7eefb4fd681bf5 0))))
    (:else
      (return 0.0))))

(fn HUDMessageRequest
  (params (Device%20Ref :type Device%20HUDMessage%20V2%20Object%20Reference) (Add :type Boolean) (MessagePayload :type HUDMessage%20Queue%20Details%20Structure))
  ;[0 1 0.9851 0.1] When we request a message, don't send it to the client immediately. Instead queue it and iterate through that queue over time. This both throttles our replication to avoid sending too many messages if we have a malicious Creator here, and also ensure ordering.
  (Utilities|SwitchHasAuthority
    (:Authority
      (if Add
        (CallFunction|AddMessage Device%20Ref MessagePayload)
        (else
          (CallFunction|RemoveDevice Device%20Ref))))))

(fn MessageDiffers
  (params (Message%201 :type HUDMessage%20Queue%20Details%20Structure) (Message%202 :type HUDMessage%20Queue%20Details%20Structure))
  (bind (_instigatorps_13_c6228da247aa3cc4ada8b28c74dad64c _messagepayload_11_b3fb8ab54557dceb753f2091d88bd21a) (Utilities|Struct|BreakHUDMessageQueueDetails Message%201))
  (bind (_message_72_7634162b4c068206f407bcac0c836150 _displaytime_37_33f9b57c45f05e666a7eefb4fd681bf5) (Utilities|Struct|BreakGameplayEventHUDMessageDetails _messagepayload_11_b3fb8ab54557dceb753f2091d88bd21a))
  (bind (_text_2_563e0da1490d74cee84ef39a5df6bb26 _message_5_ddd38afe499b5ae0e2151a99b81ee564) (Utilities|Struct|BreakLocalizablePayload _message_72_7634162b4c068206f407bcac0c836150))
  (bind (_instigatorps_13_c6228da247aa3cc4ada8b28c74dad64c_2 _messagepayload_11_b3fb8ab54557dceb753f2091d88bd21a_2) (Utilities|Struct|BreakHUDMessageQueueDetails Message%202))
  (bind (_message_72_7634162b4c068206f407bcac0c836150_2 _displaytime_37_33f9b57c45f05e666a7eefb4fd681bf5_2) (Utilities|Struct|BreakGameplayEventHUDMessageDetails _messagepayload_11_b3fb8ab54557dceb753f2091d88bd21a_2))
  (bind (_text_2_563e0da1490d74cee84ef39a5df6bb26_2 _message_5_ddd38afe499b5ae0e2151a99b81ee564_2) (Utilities|Struct|BreakLocalizablePayload _message_72_7634162b4c068206f407bcac0c836150_2))
  ;[0 0.787 1 0.2] Check the message for information that may be changed at runtime (since we've already at this point checked if the device is the same, but want to check if the contents of the device has changed)
  (return (not (Math|Boolean|ANDBoolean (== _instigatorps_13_c6228da247aa3cc4ada8b28c74dad64c _instigatorps_13_c6228da247aa3cc4ada8b28c74dad64c_2) (== _displaytime_37_33f9b57c45f05e666a7eefb4fd681bf5 _displaytime_37_33f9b57c45f05e666a7eefb4fd681bf5_2) (select (and (LocalizableMessage|IsEmptyLocalizableMessage _message_5_ddd38afe499b5ae0e2151a99b81ee564) (LocalizableMessage|IsEmptyLocalizableMessage _message_5_ddd38afe499b5ae0e2151a99b81ee564_2)) (Utilities|String|EqualExactly(String) (Utilities|String|ToString(Text) _text_2_563e0da1490d74cee84ef39a5df6bb26) (Utilities|String|ToString(Text) _text_2_563e0da1490d74cee84ef39a5df6bb26_2)) (== _message_5_ddd38afe499b5ae0e2151a99b81ee564 _message_5_ddd38afe499b5ae0e2151a99b81ee564_2))))))

(fn UpdateControllingDevice
  (params (New%20Controlling%20Device :type Device%20HUDMessage%20V2%20Object%20Reference) (MessagePayload :type HUDMessage%20Queue%20Details%20Structure))
  (bind (_instigatorps_13_c6228da247aa3cc4ada8b28c74dad64c _messagepayload_11_b3fb8ab54557dceb753f2091d88bd21a) (Utilities|Struct|BreakHUDMessageQueueDetails (Variables|Default|GetServer%20Cached%20Details)))
  (bind _messagepayload (Variables|GetMessagePayload))
  (if (or (Utilities|IsValid New%20Controlling%20Device :!class /Script/Engine.KismetSystemLibrary) (!= New%20Controlling%20Device (Variables|Default|GetControlling%20Device)))
    (LocalVariables|SetOld%20Controlling%20Device (Variables|Default|GetControlling%20Device))
    (Variables|Default|SetControlling%20Device (Variables|GetNew%20Controlling%20Device)) ; Need to set Controlling device\nbefore adding old controlling device\nto queue (if the component is current, it won't queue\nunless it can queue multiple times)
    (if (and (> (Server|GetRemainingDisplayTime) (Variables|Default|GetReaddToQueueTime)) (!= (LocalVariables|GetOld%20Controlling%20Device) (Variables|Default|GetControlling%20Device)))
      (Server|AddDeviceToQueue (LocalVariables|GetOld%20Controlling%20Device) (Variables|Default|GetServer%20Cached%20Details))
      (else
        _))
    (Variables|Default|SetLastDisplayedTime (Utilities|Time|GetGameTimeinSeconds))
    (if (!= (LocalVariables|GetOld%20Controlling%20Device) (Variables|Default|GetControlling%20Device))
      _
      (elif (CallFunction|MessageDiffers (Variables|Default|GetServer%20Cached%20Details) (Variables|GetMessagePayload))
        _
        (else
          (Utilities|ShouldShowifShowing
            (:Reshow
              _)
            (:Extend%20Timer
              (Variables|Default|SetServer%20Cached%20Details _messagepayload)
              (CallFunction|StartTimerForControlSwitch) ; Message is the same, but has been re-requested.\nReset the display timer so it'll stay on based on the new \nrequest time.
            )))))
    (Variables|Default|SetServer%20Cached%20Details _messagepayload)
    (bind _controlling_device (Variables|Default|GetControlling%20Device)
      (:then
        (CallFunction|UpdateClientNextFrame (Variables|/CreativeCoreDevices/Device_HUDMessage_V2.Device_HUDMessage_V2_C|GetMessageGUID _controlling_device) _messagepayload_11_b3fb8ab54557dceb753f2091d88bd21a _instigatorps_13_c6228da247aa3cc4ada8b28c74dad64c))
      (:else
        (CallFunction|UpdateClientNextFrame) ; No controlling device, tell the client to update to empty.
      ))))

(fn UpdateControllingDeviceFromQueue
  (bind _check_details_array (LocalVariables|GetCheck%20Details%20Array))
  (bind (_output_timeadded_2_6d40806d4734592914ecb18c9b114de8 _output_instigatorps_13_c6228da247aa3cc4ada8b28c74dad64c) (Utilities|Struct|BreakHUDMessageQueueDetails (Utilities|Array|Get(acopy) _check_details_array)))
  (bind _checkarray (LocalVariables|GetCheckArray))
  (bind _output (Utilities|Array|Get(acopy) _checkarray))
  (bind _output_timeadded_2_6d40806d4734592914ecb18c9b114de8_2 (Utilities|Struct|BreakHUDMessageQueueDetails (Utilities|Array|Get(aref) (LocalVariables|GetCheck%20Details%20Array))))
  ;[0.1455 1 0.8701 0.25] Devices that have spawned a modal when there's already a modal active are added into a queue. This function pops the first device from the queue and sets it to the new controlling device.
  (for (_array_element _array_index) (Variables|Default|GetQueued%20Devices)
    (bind (_array_element_device_4_1d5858c6423b2ef1df5985976caeba50 _array_element_message_11_5f74f8bd407bc035e95aaf8e0557f0a0) (Utilities|Struct|BreakHUDMessageQueue _array_element))
    (LocalVariables|SetCheckArray _array_element_device_4_1d5858c6423b2ef1df5985976caeba50)
    (LocalVariables|SetCheck%20Details%20Array _array_element_message_11_5f74f8bd407bc035e95aaf8e0557f0a0)
    (LocalVariables|SetCurrentPriorityCheck _array_index)
    (while (and (Utilities|Array|IsNotEmpty (LocalVariables|GetCheckArray)) (Utilities|Array|IsNotEmpty (LocalVariables|GetCheck%20Details%20Array)))
      (Utilities|IsValid _output
        (:Is%20Valid
          (Utilities|FlowControl|Sequence
            (:then_0
              (if (or (< (- (Fortnite|Sync|Time|GetServerWorldTimeSeconds) _output_timeadded_2_6d40806d4734592914ecb18c9b114de8_2) (Variables|/CreativeCoreDevices/Device_HUDMessage_V2.Device_HUDMessage_V2_C|GetQueueTimeout (Utilities|Array|Get(acopy) (LocalVariables|GetCheckArray)))) (not (Variables|/CreativeCoreDevices/Device_HUDMessage_V2.Device_HUDMessage_V2_C|GetQueueTimeout_Override (Utilities|Array|Get(acopy) (LocalVariables|GetCheckArray)))))
                (bind _output_get (LocalVariables|SetCheck%20Device _output))
                (if (Variables|/CreativeCoreDevices/Device_HUDMessage_V2.Device_HUDMessage_V2_C|GetRe-Evaluate%20Messages%20on%20Show _output_get)
                  (bind _asplayer_controller (Utilities|Casting|CastToPlayerController (Components|GetOwner)))
                  (Utilities|IsValid _output_instigatorps_13_c6228da247aa3cc4ada8b28c74dad64c
                    (:Is%20Valid
                      (bind _true (Class|DeviceHUDMessageV2|IsRelevantForPlayer (LocalVariables|GetCheck%20Device) _asplayer_controller true (PlayerState|GetPlayerController _output_instigatorps_13_c6228da247aa3cc4ada8b28c74dad64c)))
                      (if _true
                        _)))
                  (else
                    _))
                (Server|UpdateControllingDevice (LocalVariables|GetCheck%20Device) (Utilities|Array|Get(acopy) (LocalVariables|GetCheck%20Details%20Array)))
                ;;[0.03219 1 0.3023 0.2] Since we've found our device, lets pop it off our stacks and update the queue.
                (Utilities|Array|RemoveIndex (LocalVariables|GetCheckArray))
                (Utilities|Array|RemoveIndex (LocalVariables|GetCheck%20Details%20Array))
                (Utilities|Array|SetArrayElem (Variables|Default|GetQueued%20Devices) (LocalVariables|GetCurrentPriorityCheck) (Utilities|Struct|MakeHUDMessageQueue (LocalVariables|GetCheckArray) (LocalVariables|GetCheck%20Details%20Array)) false)
                (return false)))
            (:then_1
              ;;[1 0.2147 0 0.1] This device has been checked, remove it.
              (Utilities|Array|RemoveIndex (LocalVariables|GetCheckArray))
              (Utilities|Array|RemoveIndex (LocalVariables|GetCheck%20Details%20Array)))))))
    (Utilities|Array|SetArrayElem (Variables|Default|GetQueued%20Devices) (LocalVariables|GetCurrentPriorityCheck)))
  (Server|UpdateControllingDevice 0)
  (return true))
