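# Container widget for the HUD Message device that hosts, positions and animates a message widget from the supplied message details.
# UpdateFromContainer async-loads a differing widget class or refreshes the current one, deferring updates that arrive during loading or outro.
# Updates feed details and text style into an MVVM view model, then set anchors and viewport-clamped offsets.
# Enum-selected animations play forward for intros and reversed for outros, with a timer collapsing the widget afterward.
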WidgetBlueprint /CreativeCoreDevices/HUDMessage/HUDMessage_AnimationContainer
animations [
 Display {
  binding_count=1
  length_seconds=0
  object=&self:Display
  sequence {bindings [ParentPanel {
     id=#1
     possessed_class=CanvasPanel
     tracks [
      {
       type=MovieScene2DTransformTrack
       sections [{
         type=MovieScene2DTransformSection
         end_frame=0
         start_frame=0
         channels [
          Translation.X {type=MovieSceneScriptingFloatChannel, const [0,2]}
          Translation.Y {type=MovieSceneScriptingFloatChannel, const [0,2]}
          Angle {type=MovieSceneScriptingFloatChannel, const [0,2]}
          Scale.X {type=MovieSceneScriptingFloatChannel, const [1,2]}
          Scale.Y {type=MovieSceneScriptingFloatChannel, const [1,2]}
          Shear.X {type=MovieSceneScriptingFloatChannel, const [0,2]}
          Shear.Y {type=MovieSceneScriptingFloatChannel, const [0,2]}
         ]
        }]
      }
      {
       type=MovieSceneFloatTrack
       display_name=Render Opacity
       sections [{type=MovieSceneFloatSection, end_frame=0, start_frame=0, channels [None {type=MovieSceneScriptingFloatChannel, const [1,1]}]}]
      }
     ]
    }]}
 }
 Bounce {
  binding_count=1
  length_seconds=0.35
  object=&self:Bounce
  sequence {
   bindings [ParentPanel {
     id=#2
     display_name=CanvasPanel_21
     possessed_class=CanvasPanel
     tracks [
      {
       type=MovieSceneFloatTrack
       display_name=Render Opacity
       sections [{type=MovieSceneFloatSection, end_frame=10, start_frame=0, channels [None {type=MovieSceneScriptingFloatChannel, keys [[0,0],[7,1]]}]}]
      }
      {
       type=MovieScene2DTransformTrack
       sections [{
         type=MovieScene2DTransformSection
         end_frame=10
         start_frame=0
         channels [
          Translation.X {type=MovieSceneScriptingFloatChannel}
          Translation.Y {type=MovieSceneScriptingFloatChannel}
          Angle {type=MovieSceneScriptingFloatChannel}
          Scale.X {type=MovieSceneScriptingFloatChannel, keys [[0,0],[7,1.5],[9,1],[10.5,1]]}
          Scale.Y {type=MovieSceneScriptingFloatChannel, keys [[0,0],[7,1.5],[9,1],[10.5,1]]}
          Shear.X {type=MovieSceneScriptingFloatChannel, const [0,1]}
          Shear.Y {type=MovieSceneScriptingFloatChannel, const [0,1]}
         ]
        }]
      }
     ]
    }]
   display_rate {numerator=30}
  }
 }
 Zoom {
  binding_count=1
  length_seconds=0.35
  object=&self:Zoom
  sequence {bindings [ParentPanel {
     id=#2
     display_name=CanvasPanel_21
     possessed_class=CanvasPanel
     tracks [
      {
       type=MovieSceneFloatTrack
       display_name=Render Opacity
       sections [{type=MovieSceneFloatSection, end_frame=6, start_frame=0, channels [None {type=MovieSceneScriptingFloatChannel, keys [[0,0],[3.8673,1]]}]}]
      }
      {
       type=MovieScene2DTransformTrack
       sections [{
         type=MovieScene2DTransformSection
         end_frame=7
         start_frame=0
         channels [
          Translation.X {type=MovieSceneScriptingFloatChannel}
          Translation.Y {type=MovieSceneScriptingFloatChannel}
          Angle {type=MovieSceneScriptingFloatChannel}
          Scale.X {type=MovieSceneScriptingFloatChannel, keys [[0,0],[7,1]]}
          Scale.Y {type=MovieSceneScriptingFloatChannel, keys [[0,0],[7,1]]}
          Shear.X {type=MovieSceneScriptingFloatChannel, const [0,1]}
          Shear.Y {type=MovieSceneScriptingFloatChannel, const [0,1]}
         ]
        }]
      }
     ]
    }]}
 }
 ReverseZoom {
  binding_count=1
  length_seconds=0.35
  object=&self:ReverseZoom
  sequence {bindings [ParentPanel {
     id=#2
     display_name=CanvasPanel_21
     possessed_class=CanvasPanel
     tracks [
      {
       type=MovieSceneFloatTrack
       display_name=Render Opacity
       sections [{type=MovieSceneFloatSection, end_frame=7, start_frame=0, channels [None {type=MovieSceneScriptingFloatChannel, keys [[0,0],[7,1]]}]}]
      }
      {
       type=MovieScene2DTransformTrack
       sections [{
         type=MovieScene2DTransformSection
         end_frame=7
         start_frame=0
         channels [
          Translation.X {type=MovieSceneScriptingFloatChannel}
          Translation.Y {type=MovieSceneScriptingFloatChannel}
          Angle {type=MovieSceneScriptingFloatChannel}
          Scale.X {type=MovieSceneScriptingFloatChannel, keys [[0,2],[7,1]]}
          Scale.Y {type=MovieSceneScriptingFloatChannel, keys [[0,2],[7,1]]}
          Shear.X {type=MovieSceneScriptingFloatChannel, const [0,1]}
          Shear.Y {type=MovieSceneScriptingFloatChannel, const [0,1]}
         ]
        }]
      }
     ]
    }]}
 }
 Fade {
  binding_count=1
  length_seconds=0.35
  object=&self:Fade
  sequence {bindings [ParentPanel {
     id=#3
     possessed_class=CanvasPanel
     tracks [
      {
       type=MovieSceneFloatTrack
       display_name=Render Opacity
       sections [{type=MovieSceneFloatSection, end_frame=6, start_frame=0, channels [None {type=MovieSceneScriptingFloatChannel, keys [[0,0],[6.7927,1]]}]}]
      }
      {
       type=MovieScene2DTransformTrack
       sections [{
         type=MovieScene2DTransformSection
         end_frame=7
         start_frame=0
         channels [
          Translation.X {type=MovieSceneScriptingFloatChannel, const [0,2]}
          Translation.Y {type=MovieSceneScriptingFloatChannel, const [0,2]}
          Angle {type=MovieSceneScriptingFloatChannel, const [0,2]}
          Scale.X {type=MovieSceneScriptingFloatChannel, const [1,2]}
          Scale.Y {type=MovieSceneScriptingFloatChannel, const [1,2]}
          Shear.X {type=MovieSceneScriptingFloatChannel, const [0,2]}
          Shear.Y {type=MovieSceneScriptingFloatChannel, const [0,2]}
         ]
        }]
      }
     ]
    }]}
 }
 FadeZoom {
  binding_count=1
  length_seconds=0.35
  object=&self:FadeZoom
  sequence {bindings [ParentPanel {
     id=#3
     possessed_class=CanvasPanel
     tracks [
      {
       type=MovieSceneFloatTrack
       display_name=Render Opacity
       sections [{type=MovieSceneFloatSection, end_frame=7, start_frame=0, channels [None {type=MovieSceneScriptingFloatChannel, keys [[0,0],[7,1]]}]}]
      }
      {
       type=MovieScene2DTransformTrack
       sections [{
         type=MovieScene2DTransformSection
         end_frame=7
         start_frame=0
         channels [
          Translation.X {type=MovieSceneScriptingFloatChannel}
          Translation.Y {type=MovieSceneScriptingFloatChannel}
          Angle {type=MovieSceneScriptingFloatChannel}
          Scale.X {type=MovieSceneScriptingFloatChannel, keys [[0,0],[7,1]]}
          Scale.Y {type=MovieSceneScriptingFloatChannel, keys [[0,0],[7,1]]}
          Shear.X {type=MovieSceneScriptingFloatChannel}
          Shear.Y {type=MovieSceneScriptingFloatChannel}
         ]
        }]
      }
     ]
    }]}
 }
 Slow_Zoom {
  binding_count=1
  length_seconds=1
  object=&self:Slow_Zoom
  sequence {bindings [ParentPanel {
     id=#2
     display_name=CanvasPanel_21
     possessed_class=CanvasPanel
     tracks [
      {
       type=MovieSceneFloatTrack
       display_name=Render Opacity
       sections [{type=MovieSceneFloatSection, end_frame=20, start_frame=0, channels [None {type=MovieSceneScriptingFloatChannel, keys [[0,0],[20,1]]}]}]
      }
      {
       type=MovieScene2DTransformTrack
       sections [{
         type=MovieScene2DTransformSection
         end_frame=20
         start_frame=0
         channels [
          Translation.X {type=MovieSceneScriptingFloatChannel}
          Translation.Y {type=MovieSceneScriptingFloatChannel}
          Angle {type=MovieSceneScriptingFloatChannel}
          Scale.X {type=MovieSceneScriptingFloatChannel, keys [[0,0],[20,1]]}
          Scale.Y {type=MovieSceneScriptingFloatChannel, keys [[0,0],[20,1]]}
          Shear.X {type=MovieSceneScriptingFloatChannel}
          Shear.Y {type=MovieSceneScriptingFloatChannel}
         ]
        }]
      }
     ]
    }]}
 }
 Slow_ReverseZoom {
  binding_count=1
  length_seconds=1
  object=&self:Slow_ReverseZoom
  sequence {bindings [ParentPanel {
     id=#2
     display_name=CanvasPanel_21
     possessed_class=CanvasPanel
     tracks [
      {
       type=MovieSceneFloatTrack
       display_name=Render Opacity
       sections [{type=MovieSceneFloatSection, end_frame=20, start_frame=0, channels [None {type=MovieSceneScriptingFloatChannel, keys [[0,0],[20,1]]}]}]
      }
      {
       type=MovieScene2DTransformTrack
       sections [{
         type=MovieScene2DTransformSection
         end_frame=20
         start_frame=0
         channels [
          Translation.X {type=MovieSceneScriptingFloatChannel}
          Translation.Y {type=MovieSceneScriptingFloatChannel}
          Angle {type=MovieSceneScriptingFloatChannel}
          Scale.X {type=MovieSceneScriptingFloatChannel, keys [[0,2],[20,1]]}
          Scale.Y {type=MovieSceneScriptingFloatChannel, keys [[0,2],[20,1]]}
          Shear.X {type=MovieSceneScriptingFloatChannel}
          Shear.Y {type=MovieSceneScriptingFloatChannel}
         ]
        }]
      }
     ]
    }]}
 }
 Slow_Fade {
  binding_count=1
  length_seconds=1
  object=&self:Slow_Fade
  sequence {bindings [ParentPanel {
     id=#3
     possessed_class=CanvasPanel
     tracks [
      {
       type=MovieSceneFloatTrack
       display_name=Render Opacity
       sections [{type=MovieSceneFloatSection, end_frame=20, start_frame=0, channels [None {type=MovieSceneScriptingFloatChannel, keys [[0,0],[20,1]]}]}]
      }
      {
       type=MovieScene2DTransformTrack
       sections [{
         type=MovieScene2DTransformSection
         end_frame=20
         start_frame=0
         channels [
          Translation.X {type=MovieSceneScriptingFloatChannel, const [0,2]}
          Translation.Y {type=MovieSceneScriptingFloatChannel, const [0,2]}
          Angle {type=MovieSceneScriptingFloatChannel, const [0,2]}
          Scale.X {type=MovieSceneScriptingFloatChannel, const [1,2]}
          Scale.Y {type=MovieSceneScriptingFloatChannel, const [1,2]}
          Shear.X {type=MovieSceneScriptingFloatChannel, const [0,2]}
          Shear.Y {type=MovieSceneScriptingFloatChannel, const [0,2]}
         ]
        }]
      }
     ]
    }]}
 }
 Slow_FadeZoom {
  binding_count=1
  length_seconds=1
  object=&self:Slow_FadeZoom
  sequence {bindings [ParentPanel {
     id=#3
     possessed_class=CanvasPanel
     tracks [
      {
       type=MovieSceneFloatTrack
       display_name=Render Opacity
       sections [{type=MovieSceneFloatSection, end_frame=20, start_frame=0, channels [None {type=MovieSceneScriptingFloatChannel, keys [[0,0],[20,1]]}]}]
      }
      {
       type=MovieScene2DTransformTrack
       sections [{
         type=MovieScene2DTransformSection
         end_frame=20
         start_frame=0
         channels [
          Translation.X {type=MovieSceneScriptingFloatChannel, const [0,2]}
          Translation.Y {type=MovieSceneScriptingFloatChannel, const [0,2]}
          Angle {type=MovieSceneScriptingFloatChannel, const [0,2]}
          Scale.X {type=MovieSceneScriptingFloatChannel, keys [[0,0],[20,1]]}
          Scale.Y {type=MovieSceneScriptingFloatChannel, keys [[0,0],[20,1]]}
          Shear.X {type=MovieSceneScriptingFloatChannel, const [0,2]}
          Shear.Y {type=MovieSceneScriptingFloatChannel, const [0,2]}
         ]
        }]
      }
     ]
    }]}
 }
 SlideFromRight {
  binding_count=1
  length_seconds=0.35
  object=&self:SlideFromRight
  sequence {bindings [ParentPanel {
     id=#3
     possessed_class=CanvasPanel
     tracks [
      {
       type=MovieSceneFloatTrack
       display_name=Render Opacity
       sections [{type=MovieSceneFloatSection, end_frame=7, start_frame=0, channels [None {type=MovieSceneScriptingFloatChannel, keys [[0,0],[7,1]]}]}]
      }
      {
       type=MovieScene2DTransformTrack
       sections [{
         type=MovieScene2DTransformSection
         end_frame=7
         start_frame=0
         channels [
          Translation.X {type=MovieSceneScriptingFloatChannel, keys [[0,300],[7,0]]}
          Translation.Y {type=MovieSceneScriptingFloatChannel, const [0,1]}
          Angle {type=MovieSceneScriptingFloatChannel, const [0,1]}
          Scale.X {type=MovieSceneScriptingFloatChannel, const [1,1]}
          Scale.Y {type=MovieSceneScriptingFloatChannel, const [1,1]}
          Shear.X {type=MovieSceneScriptingFloatChannel, const [0,1]}
          Shear.Y {type=MovieSceneScriptingFloatChannel, const [0,1]}
         ]
        }]
      }
     ]
    }]}
 }
 SlideFromLeft {
  binding_count=1
  length_seconds=0.35
  object=&self:SlideFromLeft
  sequence {bindings [ParentPanel {
     id=#3
     possessed_class=CanvasPanel
     tracks [
      {
       type=MovieSceneFloatTrack
       display_name=Render Opacity
       sections [{type=MovieSceneFloatSection, end_frame=7, start_frame=0, channels [None {type=MovieSceneScriptingFloatChannel, keys [[0,0],[7,1]]}]}]
      }
      {
       type=MovieScene2DTransformTrack
       sections [{
         type=MovieScene2DTransformSection
         end_frame=7
         start_frame=0
         channels [
          Translation.X {type=MovieSceneScriptingFloatChannel, keys [[0,-300],[7,0]]}
          Translation.Y {type=MovieSceneScriptingFloatChannel, const [0,2]}
          Angle {type=MovieSceneScriptingFloatChannel, const [0,1]}
          Scale.X {type=MovieSceneScriptingFloatChannel, const [1,1]}
          Scale.Y {type=MovieSceneScriptingFloatChannel, const [1,1]}
          Shear.X {type=MovieSceneScriptingFloatChannel, const [0,1]}
          Shear.Y {type=MovieSceneScriptingFloatChannel, const [0,1]}
         ]
        }]
      }
     ]
    }]}
 }
 SlideFromTop {
  binding_count=1
  length_seconds=0.35
  object=&self:SlideFromTop
  sequence {bindings [ParentPanel {
     id=#3
     possessed_class=CanvasPanel
     tracks [
      {
       type=MovieSceneFloatTrack
       display_name=Render Opacity
       sections [{type=MovieSceneFloatSection, end_frame=7, start_frame=0, channels [None {type=MovieSceneScriptingFloatChannel, keys [[0,0],[7,1]]}]}]
      }
      {
       type=MovieScene2DTransformTrack
       sections [{
         type=MovieScene2DTransformSection
         end_frame=7
         start_frame=0
         channels [
          Translation.X {type=MovieSceneScriptingFloatChannel, const [0,2]}
          Translation.Y {type=MovieSceneScriptingFloatChannel, keys [[0,-300],[7,0]]}
          Angle {type=MovieSceneScriptingFloatChannel, const [0,1]}
          Scale.X {type=MovieSceneScriptingFloatChannel, const [1,1]}
          Scale.Y {type=MovieSceneScriptingFloatChannel, const [1,1]}
          Shear.X {type=MovieSceneScriptingFloatChannel, const [0,1]}
          Shear.Y {type=MovieSceneScriptingFloatChannel, const [0,1]}
         ]
        }]
      }
     ]
    }]}
 }
 SlideFromBottom {
  binding_count=1
  length_seconds=0.35
  object=&self:SlideFromBottom
  sequence {bindings [ParentPanel {
     id=#3
     possessed_class=CanvasPanel
     tracks [
      {
       type=MovieSceneFloatTrack
       display_name=Render Opacity
       sections [{type=MovieSceneFloatSection, end_frame=7, start_frame=0, channels [None {type=MovieSceneScriptingFloatChannel, keys [[0,0],[7,1]]}]}]
      }
      {
       type=MovieScene2DTransformTrack
       sections [{
         type=MovieScene2DTransformSection
         end_frame=7
         start_frame=0
         channels [
          Translation.X {type=MovieSceneScriptingFloatChannel, const [0,2]}
          Translation.Y {type=MovieSceneScriptingFloatChannel, keys [[0,300],[7,0]]}
          Angle {type=MovieSceneScriptingFloatChannel, const [0,1]}
          Scale.X {type=MovieSceneScriptingFloatChannel, const [1,1]}
          Scale.Y {type=MovieSceneScriptingFloatChannel, const [1,1]}
          Shear.X {type=MovieSceneScriptingFloatChannel, const [0,1]}
          Shear.Y {type=MovieSceneScriptingFloatChannel, const [0,1]}
         ]
        }]
      }
     ]
    }]}
 }
]
graphs [
 ConvertAnchorEnumToAnchor {nodes [
   n1 = ConvertAnchorEnumToAnchor!() -> (then, Anchor:EUIAnchor Enum=NewEnumerator0) {is=title K2Node_FunctionEntry}
   > Return Node!(
    execute
    Min:Vector 2D Structure=Select!(
     NewEnumerator0:Vector 2D Structure=(X=0,Y=0)
     NewEnumerator1:Vector 2D Structure=(X=0.5,Y=0)
     NewEnumerator2:Vector 2D Structure=(X=1,Y=0)
     NewEnumerator3:Vector 2D Structure=(X=0,Y=0.5)
     NewEnumerator4:Vector 2D Structure=(X=0.5,Y=0.5)
     NewEnumerator5:Vector 2D Structure=(X=1,Y=0.5)
     NewEnumerator6:Vector 2D Structure=(X=0,Y=1)
     NewEnumerator7:Vector 2D Structure=(X=0.5,Y=1)
     NewEnumerator8:Vector 2D Structure=(X=1,Y=1)
     NewEnumerator9:Vector 2D Structure
     NewEnumerator10:Vector 2D Structure
     NewEnumerator11:Vector 2D Structure=(X=0.5,Y=0)
     NewEnumerator12:Vector 2D Structure=(X=1,Y=0)
     NewEnumerator13:Vector 2D Structure
     NewEnumerator14:Vector 2D Structure=(X=0,Y=0.5)
     NewEnumerator15:Vector 2D Structure=(X=0,Y=1)
     Index:EUIAnchor Enum=@n1
    ) -> (ReturnValue:Vector 2D Structure)
    Max:Vector 2D Structure=Select(
     NewEnumerator0=(X=0,Y=0)
     NewEnumerator1=(X=0.5,Y=0)
     NewEnumerator2=(X=1,Y=0)
     NewEnumerator3=(X=0,Y=0.5)
     NewEnumerator4=(X=0.5,Y=0.5)
     NewEnumerator5=(X=1,Y=0.5)
     NewEnumerator6=(X=0,Y=1)
     NewEnumerator7=(X=0.5,Y=1)
     NewEnumerator8=(X=1,Y=1)
     NewEnumerator9=(X=1,Y=1)
     NewEnumerator10=(X=0,Y=1)
     NewEnumerator11=(X=0.5,Y=1)
     NewEnumerator12=(X=1,Y=1)
     NewEnumerator13=(X=1,Y=0)
     NewEnumerator14=(X=1,Y=0.5)
     NewEnumerator15=(X=1,Y=1)
     Index=@n1
    )
    Alignment:Vector 2D Structure=Select(
     NewEnumerator0=(X=0,Y=0)
     NewEnumerator1=(X=0.5,Y=0)
     NewEnumerator2=(X=1,Y=0)
     NewEnumerator3=(X=0,Y=0.5)
     NewEnumerator4=(X=0.5,Y=0.5)
     NewEnumerator5=(X=1,Y=0.5)
     NewEnumerator6=(X=0,Y=1)
     NewEnumerator7=(X=0.5,Y=1)
     NewEnumerator8=(X=1,Y=1)
     NewEnumerator9=(X=0.5,Y=0.5)
     NewEnumerator10=(X=0,Y=0.5)
     NewEnumerator11=(X=0.5,Y=0.5)
     NewEnumerator12=(X=1,Y=0.5)
     NewEnumerator13=(X=0.5,Y=0)
     NewEnumerator14=(X=0.5,Y=0.5)
     NewEnumerator15=(X=0.5,Y=1)
     Index=@n1
    )
   )
  ]}
 EventGraph {nodes [
   Update and Display Widget() {is=title K2Node_CustomEvent}
   > Update View Model()
   > UpdateMessageVisibility!(execute, self, On:bool=true, Instant:bool=false) -> (then)
   > DelayUntilNextTick!(execute) -> (then)
   > Update Screen Position()
   Load Widget() {is=title K2Node_CustomEvent}
   > Set IsLoading(IsLoading:bool=true) -> (Output_Get:bool)
   > n1 = Get!(execute:Exec) -> (then:Exec, else:Exec, LoadedWidget:User Widget Object Reference)
   > RemoveFromParent(self=@n1)
   > n2 = Set LoadedWidget(LoadedWidget:User Widget Object Reference) -> (Output_Get:User Widget Object Reference)
   n3 = CreateWidgetAsync!(
    execute:Exec=[@n1.else, @n2.then]
    UserWidgetSoftClass:User Widget Soft Class Reference=Break Gameplay Event HUDMessage Details!(GameplayEvent_HUDMessageDetails:"Gameplay Event HUDMessage Details Structure (by ref)"=Get Current Details() -> (Current Details:Gameplay Event HUDMessage Details Structure)) -> (WidgetClass_6_#4:User Widget Soft Class Reference) {is=title K2Node_BreakStruct}
    OwningPlayer:Player Controller Object Reference
    bSuspendInputUntilComplete:bool=false
   ) -> (then:Exec, OnComplete:Exec, OnFailure:Exec, UserWidget:User Widget Object Reference) {is=title K2Node_AsyncAction}
   Initialize Loaded Widget!(execute=@n3.OnComplete, self, LoadedWidget:User Widget Object Reference=@n3) -> (then)
   > DelayUntilNextTick()
   > Set IsLoading(IsLoading=false)
   > n4 = Branch(Condition=Get NeedsRecheck() -> (NeedsRecheck:bool))
   > Set NeedsRecheck(NeedsRecheck:bool=false) -> (Output_Get:bool)
   > UpdateFromContainer!(execute, self, New Details:Gameplay Event HUDMessage Details Structure=Get Current Details()) -> (then)
   Update and Display Widget(execute=@n4.else) {is=title K2Node_CallFunction}
   n5 = UpdateFromContainer!() -> (OutputDelegate, then, New Details:Gameplay Event HUDMessage Details Structure) {is=title K2Node_CustomEvent}
   > n6 = Set Current Details(Current Details:Gameplay Event HUDMessage Details Structure=@n5.New Details) -> (Output_Get:Gameplay Event HUDMessage Details Structure)
   > n7 = Message State Check!(execute:Exec, New Details:Gameplay Event HUDMessage Details Structure=@n6) -> (Yes:Exec, Different Widget:Exec, CurrentlyLoading:Exec, CurrentlyClosing:Exec, Invalid Widget:Exec) {is=title K2Node_MacroInstance}
   Set NeedsRecheck(execute=[@n7.CurrentlyClosing, @n7.CurrentlyLoading], NeedsRecheck=true)
   > Set DesiredVisibility(DesiredVisibility:bool=true) -> (Output_Get:bool)
   Load Widget(execute=@n7.Different Widget) {is=title K2Node_CallFunction}
   Update and Display Widget(execute=@n7.Yes)
   K2Node_Event_1 = UpdateMessageVisibility!() -> (OutputDelegate, then, On:bool=~, Instant:bool=~) {is=title K2Node_CustomEvent}
   > n8 = Branch(Condition=@K2Node_Event_1.On)
   > Show!(execute, self, Instant:bool=@K2Node_Event_1.Instant) -> (then)
   Hide!(execute=@n8.else, self, Instant:bool=@K2Node_Event_1.Instant) -> (then)
  ]}
 GetAnim {nodes [
   n1 = Inputs!() -> (" ":Exec, Index:EHUDMessageAnimation Enum=NewEnumerator0)
   n2 = Get FadeZoom() -> (FadeZoom:Widget Animation Object Reference)
   n3 = Get Zoom() -> (Zoom:Widget Animation Object Reference)
   n4 = Get Slow_Fade() -> (Slow_Fade:Widget Animation Object Reference)
   n5 = Get Slow_ReverseZoom() -> (Slow_ReverseZoom:Widget Animation Object Reference)
   n6 = Get Slow_Zoom() -> (Slow_Zoom:Widget Animation Object Reference)
   n7 = Get ReverseZoom() -> (ReverseZoom:Widget Animation Object Reference)
   n8 = Get Bounce() -> (Bounce:Widget Animation Object Reference)
   n9 = Get Fade() -> (Fade:Widget Animation Object Reference)
   n10 = Get Display() -> (Display:Widget Animation Object Reference)
   n11 = Get SlideFromBottom() -> (SlideFromBottom:Widget Animation Object Reference)
   n12 = Get SlideFromTop() -> (SlideFromTop:Widget Animation Object Reference)
   n13 = Get SlideFromLeft() -> (SlideFromLeft:Widget Animation Object Reference)
   n14 = Get SlideFromRight() -> (SlideFromRight:Widget Animation Object Reference)
   n15 = Get Slow_FadeZoom() -> (Slow_FadeZoom:Widget Animation Object Reference)
   n16 = Local Wildcard!() -> (Variable:Widget Animation Object Reference) {is=title K2Node_TemporaryVariable}
   > "Set Widget Animation Object Reference (by ref)"!(
    execute:Exec
    Target:"Widget Animation Object Reference (by ref)"=@n16
    Value:Widget Animation Object Reference=Select!(
     NewEnumerator0:Widget Animation Object Reference=@n10
     NewEnumerator1:Widget Animation Object Reference=@n3
     NewEnumerator2:Widget Animation Object Reference=@n2
     NewEnumerator4:Widget Animation Object Reference=@n9
     NewEnumerator5:Widget Animation Object Reference=@n7
     NewEnumerator6:Widget Animation Object Reference=@n8
     NewEnumerator7:Widget Animation Object Reference=@n6
     NewEnumerator8:Widget Animation Object Reference=@n15
     NewEnumerator9:Widget Animation Object Reference=@n4
     NewEnumerator10:Widget Animation Object Reference=@n5
     NewEnumerator11:Widget Animation Object Reference=@n12
     NewEnumerator12:Widget Animation Object Reference=@n11
     NewEnumerator13:Widget Animation Object Reference=@n13
     NewEnumerator14:Widget Animation Object Reference=@n14
     Index:EHUDMessageAnimation Enum=@n1
    ) -> (ReturnValue:Widget Animation Object Reference)
   ) -> (then:Exec) {is=title K2Node_VariableSetRef}
   > n17 = Is Valid(InputObject=@n16)
   n18 = "Set Widget Animation Object Reference (by ref)"(execute=@n17.Is Not Valid, Target=@n16, Value=Get Display())
   Outputs!(
    HasAnim:Exec=[@n17.Is Valid, @n18.then]
    FoundAnim:Widget Animation Object Reference=@n16
    AnimList:Array of Widget Animation Object References=Make Array!(
     "[0]":Widget Animation Object Reference=@n10
     "[1]":Widget Animation Object Reference=@n3
     "[2]":Widget Animation Object Reference=@n2
     "[3]":Widget Animation Object Reference=@n9
     "[4]":Widget Animation Object Reference=@n7
     "[5]":Widget Animation Object Reference=@n8
     "[6]":Widget Animation Object Reference=@n6
     "[7]":Widget Animation Object Reference=@n15
     "[8]":Widget Animation Object Reference=@n4
     "[9]":Widget Animation Object Reference=@n5
     "[10]":Widget Animation Object Reference=@n12
     "[11]":Widget Animation Object Reference=@n11
     "[12]":Widget Animation Object Reference=@n13
     "[13]":Widget Animation Object Reference=@n14
    ) -> (Array:Array of Widget Animation Object References)
   )
  ]}
 GetTextStyle {nodes [
   n1 = GetTextStyle!() -> (then, GameplayEvent_HUDMessageDetails:"Gameplay Event HUDMessage Details Structure (by ref)") {is=title K2Node_FunctionEntry}
   n2 = Break Gameplay Event HUDMessage Details!(GameplayEvent_HUDMessageDetails=@n1) -> (TextStyleSetOverride_45_#5:Data Table Object Reference)
   > n3 = Is Valid(InputObject=@n2)
   n4 = GetMinigameSettingsMachine!(ContextActor:Actor Object Reference=GetFirstLocalFortPlayerController!(bIncludeNonSpectators:bool=true, bIncludeSpectators:bool=false) -> (ReturnValue:Fort Player Controller Object Reference)) -> (ReturnValue:Fort Minigame Settings Building Object Reference)
   n5 = Is Valid(exec=@n3.Is Not Valid, InputObject=@n4)
   Return Node!(
    execute=@n5.Is Not Valid
    Text Style Set Override:Data Table Object Reference=/CreativeCoreDevices/Athena/UI/InGame/Creative/PropertyEditor/StyleEditor/RichTextStyleSets/CreativeRichTextStyles_Default
   )
   > Return Node(Text Style Set Override=GetSelectedTextStyleSet!(self:Fort Minigame Settings Building Object Reference=@n4) -> (ReturnValue:Data Table Object Reference))
   Return Node(execute=@n3.Is Valid, Text Style Set Override=@n2)
  ]}
 GetViewportOffsets {nodes [
   n1 = GetViewportOffsets!() -> (then, Anchor:EUIAnchor Enum=NewEnumerator0, Offset:Vector 2D Structure) {is=title K2Node_FunctionEntry}
   > Set AnchorValue(AnchorValue:Vector 2D Structure=Select!(
     NewEnumerator0:Vector 2D Structure=(X=0,Y=0)
     NewEnumerator1:Vector 2D Structure=(X=0.5,Y=0)
     NewEnumerator2:Vector 2D Structure=(X=1,Y=0)
     NewEnumerator3:Vector 2D Structure=(X=0,Y=0.5)
     NewEnumerator4:Vector 2D Structure=(X=0.5,Y=0.5)
     NewEnumerator5:Vector 2D Structure=(X=1,Y=0.5)
     NewEnumerator6:Vector 2D Structure=(X=0,Y=1)
     NewEnumerator7:Vector 2D Structure=(X=0.5,Y=1)
     NewEnumerator8:Vector 2D Structure=(X=1,Y=1)
     NewEnumerator9:Vector 2D Structure=(X=0.5,Y=0.5)
     NewEnumerator10:Vector 2D Structure=(X=0,Y=0.5)
     NewEnumerator11:Vector 2D Structure=(X=0.5,Y=0.5)
     NewEnumerator12:Vector 2D Structure=(X=1,Y=0.5)
     NewEnumerator13:Vector 2D Structure=(X=0.5,Y=0)
     NewEnumerator14:Vector 2D Structure=(X=0.5,Y=0.5)
     NewEnumerator15:Vector 2D Structure=(X=0.5,Y=1)
     Index:EUIAnchor Enum=@n1
    ) -> (ReturnValue:Vector 2D Structure)) -> (Output_Get:Vector 2D Structure)
   n2 = GetDesiredSize!(self:Widget Object Reference=Self-Reference()) -> (ReturnValue_X:double, ReturnValue_Y:double)
   n3 = Get AnchorValue() -> (AnchorValue:Vector 2D Structure)
   n4 = BreakVector2D(InVec=@n3)
   n5 = vector2d - vector2d!(A:Vector 2D Structure=(X=1,Y=1), B:Vector 2D Structure=@n3) -> (ReturnValue_X:double, ReturnValue_Y:double) {is=title K2Node_PromotableOperator}
   > Set!(
    execute
    WidgetOffsetMinMax_X:double=float * float!(A=@n2, B=@n4, C:double=-1) -> (ReturnValue) {is=title K2Node_PromotableOperator}
    WidgetOffsetMinMax_Y:double=float * float(A=@n2.ReturnValue_Y, B=@n4.Y, C=-1)
    WidgetOffsetMinMax_Z:double=float * float!(A=@n2, B=@n5) -> (ReturnValue)
    WidgetOffsetMinMax_W:double=float * float(A=@n2.ReturnValue_Y, B=@n5.ReturnValue_Y)
   ) -> (then, Output_Get:Vector 4 Structure)
   > n6 = GetViewportWidgetGeometry!(execute) -> (then, ReturnValue:Geometry Structure)
   n7 = GetLocalSize!(Geometry:"Geometry Structure (by ref)"=@n6) -> (ReturnValue_X:double, ReturnValue_Y:double)
   > UpdateTextInputWidth!(execute, self, ScreenWidth:double=@n7, Anchor:EUIAnchor Enum=Get Anchor() -> (Anchor:EUIAnchor Enum=NewEnumerator0)) -> (then)
   n8 = Get AnchorValue()
   n9 = BreakVector2D(InVec=@n8)
   n10 = float * float!(A=@n7, B=@n9, C:double=-1) -> (ReturnValue)
   n11 = vector2d - vector2d(A=(X=1,Y=1), B=@n8)
   n12 = float * float!(A=@n7.ReturnValue_Y, B=@n11.ReturnValue_Y) -> (ReturnValue)
   n13 = float * float!(A=@n7.ReturnValue_Y, B=@n9.Y, C:double=-1) -> (ReturnValue)
   n14 = float * float!(A=@n7, B=@n11) -> (ReturnValue)
   > Set PossibleSpace(PossibleSpace:Vector 4 Structure=Vector4 - Vector4(
     A:"Vector 4 Structure (by ref)"=MakeVector4!(X:double=@n10, Y:double=@n13, Z:double=@n14, W:double=@n12) -> (ReturnValue:Vector 4 Structure)
     B:"Vector 4 Structure (by ref)"=Get WidgetOffsetMinMax() -> (WidgetOffsetMinMax:Vector 4 Structure)
    ) -> (ReturnValue:Vector 4 Structure) {is=title K2Node_PromotableOperator}) -> (Output_Get:Vector 4 Structure)
   n15 = Get!() -> (PossibleSpace_X:double, PossibleSpace_Y:double, PossibleSpace_Z:double, PossibleSpace_W:double)
   n16 = Get!() -> (Offset_X:double, Offset_Y:double)
   > Return Node!(
    execute
    ReturnValue:Vector 2D Structure=MakeVector2D(
     X="Clamp (Float)"(Value=@n16, Min=@n15, Max=@n15.PossibleSpace_Z)
     Y="Clamp (Float)"(Value=@n16.Offset_Y, Min=@n15.PossibleSpace_Y, Max=@n15.PossibleSpace_W)
    )
   )
  ]}
 Hide {nodes [
   n1 = Hide!() -> (then, Instant:bool=~) {is=title K2Node_FunctionEntry}
   > Set DesiredVisibility(DesiredVisibility=false)
   n2 = Is Valid Timer Handle!(Handle:Timer Handle Structure=Get OutroTimerHandle() -> (OutroTimerHandle:Timer Handle Structure)) -> (ReturnValue:bool)
   > Branch(Condition=OR Boolean(A=@n1, B=NOT Boolean(A=@n2)))
   n3 = Break Gameplay Event HUDMessage Details!(GameplayEvent_HUDMessageDetails=Get Current Details()) -> (
    IntroAnimation_62_#6:EHUDMessageAnimation Enum
    OutroAnimation_65_#7:EHUDMessageAnimation Enum
   )
   > n4 = Get Anim!(" ":Exec, Index:EHUDMessageAnimation Enum=@n3) -> (HasAnim:Exec, FoundAnim:Widget Animation Object Reference, AnimList:Array of Widget Animation Object References) {is=title K2Node_MacroInstance}
   > n5 = For Each Loop(Array:Array of Widget Animation Object References=@n4.AnimList) -> (Array Element:Widget Animation Object Reference)
   > StopAnimation(InAnimation=@n5)
   n6 = IsVisible()
   n7 = Branch(execute=[@n12.OutputPin, @n5.Completed], Condition=OR Boolean(A=NOT Boolean(A=@n6), B=Get Instant() -> (Instant:bool)))
   > n8 = Get Anim(Index=@n3.OutroAnimation_65_#7)
   > n9 = For Each Loop(Array=@n8.AnimList)
   > StopAnimation(InAnimation=@n9)
   Clear and Invalidate Timer by Handle(execute=[@n13.OutputPin, @n9.Completed], Handle=Get OutroTimerHandle())
   > OnOutroAnimationFinished()
   n10 = Get Anim(" "=@n7.else, Index=@n3.OutroAnimation_65_#7)
   > PlayAnimationReverse(InAnimation=@n10)
   > n11 = Set Timer by Event(Delegate=Create Event(), Time=GetEndTime(self=@n10), bLooping=false)
   > Set OutroTimerHandle(OutroTimerHandle:Timer Handle Structure=@n11) -> (Output_Get:Timer Handle Structure)
   n12 = Reroute Node()
   n13 = Reroute Node()
  ]}
 Initialize Loaded Widget {nodes [
   n1 = Initialize Loaded Widget!() -> (then, LoadedWidget:User Widget Object Reference) {is=title K2Node_FunctionEntry}
   > Set LoadedWidget(LoadedWidget=@n1)
   > AddChild(self=Get ParentPanel() -> (ParentPanel:Canvas Panel Object Reference), Content=@n1)
   n2 = SlotAsCanvasSlot!(Widget:Widget Object Reference=@n1) -> (ReturnValue:Canvas Panel Slot Object Reference)
   > Set bAutoSize!(execute, bAutoSize:bool=true, self:Canvas Panel Slot Object Reference=@n2) -> (then, Output_Get:bool)
   > SetAnchors!(
    execute
    self:Canvas Panel Slot Object Reference=@n2
    InAnchors_Minimum:Vector 2D Structure=(X=0,Y=0)
    InAnchors_Maximum:Vector 2D Structure=(X=1,Y=1)
   ) -> (then)
   > SetAlignment!(execute, self:Canvas Panel Slot Object Reference=@n2, InAlignment:Vector 2D Structure=(X=0.5,Y=0.5)) -> (then)
   > SetOffsets!(
    execute
    self:Canvas Panel Slot Object Reference=@n2
    InOffset_Left:float=0
    InOffset_Top:float=0
    InOffset_Right:float=0
    InOffset_Bottom:float=0
   ) -> (then)
   > n3 = Get!(execute:Exec) -> (then:Exec, else:Exec, SelfAsCanvasPanelSlot:Canvas Panel Slot Object Reference)
   n4 = SlotAsCanvasSlot(Widget=Self-Reference())
   Set bAutoSize(execute=[@n5.then, @n3.then], bAutoSize=true, self=@n4)
   n5 = Set SelfAsCanvasPanelSlot(execute=@n3.else, SelfAsCanvasPanelSlot:Canvas Panel Slot Object Reference=@n4) -> (Output_Get:Canvas Panel Slot Object Reference)
  ]}
 Message State Check {nodes [
   n1 = Inputs!() -> (execute:Exec, New Details:Gameplay Event HUDMessage Details Structure)
   n2 = Break Gameplay Event HUDMessage Details!(GameplayEvent_HUDMessageDetails=@n1) -> (WidgetClass_6_#4:User Widget Soft Class Reference)
   > n3 = Branch(Condition=IsValidSoftClassReference!(SoftClassReference:"Object Soft Class Reference (by ref)"=@n2) -> (ReturnValue:bool))
   n4 = Get OutroTimerHandle()
   > n5 = Branch(Condition=NOT Boolean(A=Is Timer Active by Handle!(Handle:Timer Handle Structure=@n4) -> (ReturnValue:bool)))
   > n6 = Branch(Condition=NOT Boolean(A=Get IsLoading() -> (IsLoading:bool)))
   > n7 = Get!(execute) -> (then, else, LoadedWidget:User Widget Object Reference)
   n8 = Get Class(Object=Get LoadedWidget() -> (LoadedWidget:User Widget Object Reference)) -> (ReturnValue:User Widget Class Reference)
   > n9 = Branch(Condition="Equal (SoftClassReference)"(
     A:"Object Soft Class Reference (by ref)"=Make Soft Reference!(Input:User Widget Class Reference=@n8) -> (Output:User Widget Soft Class Reference)
     B:"Object Soft Class Reference (by ref)"=@n2
    ) -> (ReturnValue:bool=~) {is=title K2Node_PromotableOperator})
   > Outputs!(
    Yes:Exec
    Different Widget:Exec=[@n9.else, @n7.else]
    CurrentlyLoading:Exec=@n6.else
    CurrentlyClosing:Exec=@n5.else
    Invalid Widget:Exec=@n3.else
   )
  ]}
 OnOutroAnimationFinished {nodes [
   OnOutroAnimationFinished() {is=title K2Node_FunctionEntry}
   > SetVisibility(InVisibility=Collapsed)
   > Clear and Invalidate Timer by Handle(Handle=Get OutroTimerHandle())
   > Branch(Condition=Get NeedsRecheck())
   > Set NeedsRecheck(NeedsRecheck=false)
   > Branch(Condition=Get DesiredVisibility() -> (DesiredVisibility:bool))
   > UpdateFromContainer(New Details=Get Current Details()) {is=title K2Node_CallFunction}
  ]}
 Reset Animation Effects {nodes [
   Reset Animation Effects() {is=title K2Node_FunctionEntry}
   n1 = Get ParentPanel()
   > SetRenderOpacity(self=@n1, InOpacity=1)
   > SetRenderTransform!(
    execute
    self:Widget Object Reference=@n1
    InTransform_Translation:Vector 2D Structure=(X=0,Y=0)
    InTransform_Scale:Vector 2D Structure=(X=1,Y=1)
    InTransform_Shear:Vector 2D Structure=(X=0,Y=0)
    InTransform_Angle:float=0
   ) -> (then)
  ]}
 Show {nodes [
   n1 = Show!() -> (then, Instant:bool=~) {is=title K2Node_FunctionEntry}
   > Set DesiredVisibility(DesiredVisibility=true)
   > SetVisibility(InVisibility=SelfHitTestInvisible)
   n2 = Break Gameplay Event HUDMessage Details!(GameplayEvent_HUDMessageDetails=Get Current Details()) -> (
    IntroAnimation_62_#6:EHUDMessageAnimation Enum
    OutroAnimation_65_#7:EHUDMessageAnimation Enum
   )
   > n3 = Get Anim(Index=@n2.OutroAnimation_65_#7)
   > n4 = For Each Loop(Array=@n3.AnimList)
   > StopAnimation(InAnimation=@n4)
   Clear and Invalidate Timer by Handle(execute=@n4.Completed, Handle=Get OutroTimerHandle())
   > n5 = Branch(Condition=NOT Boolean(A=@n1))
   > n6 = Get Anim(Index=@n2)
   > n7 = PlayAnimation(InAnimation=@n6)
   n8 = Get(execute=[@n7.then, @n10.then])
   > n9 = Cast To BPi_HUDMessage!(execute, Object=@n8) -> (then, CastFailed, AsBPi HUDMessage:BPi HUDMessage Interface) {is=title K2Node_DynamicCast}
   > PlayIntro(self:BPi HUDMessage Interface=@n9)
   n10 = Reset Animation Effects(execute=@n5.else) {is=title K2Node_CallFunction}
  ]}
 Update Screen Position {nodes [
   Update Screen Position() {is=title K2Node_FunctionEntry}
   n1 = Break Gameplay Event HUDMessage Details!(GameplayEvent_HUDMessageDetails=Get Current Details()) -> (Anchor_12_#8:EUIAnchor Enum, Offset_15_#9:Vector 2D Structure)
   > n2 = GetViewportOffsets!(execute, self, Anchor:EUIAnchor Enum=@n1, Offset:Vector 2D Structure=@n1.Offset_15_#9) -> (then, ReturnValue:Vector 2D Structure) {is=title K2Node_CallFunction}
   > n3 = ConvertAnchorEnumToAnchor!(execute, self, Anchor:EUIAnchor Enum=@n1) -> (then, Min:Vector 2D Structure, Max:Vector 2D Structure, Alignment:Vector 2D Structure) {is=title K2Node_CallFunction}
   n4 = Get SelfAsCanvasPanelSlot() -> (SelfAsCanvasPanelSlot:Canvas Panel Slot Object Reference)
   > SetAnchors(self=@n4, InAnchors_Minimum=@n3, InAnchors_Maximum=@n3.Max)
   n5 = BreakVector2D(InVec=@n2)
   n6 = vector2d * vector2d!(A:Vector 2D Structure=@n2, B:Vector 2D Structure=(X=-1,Y=-1)) -> (ReturnValue_X:double, ReturnValue_Y:double) {is=title K2Node_PromotableOperator}
   > SetOffsets(self=@n4, InOffset_Left=@n5, InOffset_Top=@n5.Y, InOffset_Right=@n6, InOffset_Bottom=@n6.ReturnValue_Y)
   > SetAlignment(self=@n4, InAlignment=@n3.Alignment)
  ]}
 Update View Model {nodes [
   Update View Model() {is=title K2Node_FunctionEntry}
   > Get()
   > n1 = Get!(execute) -> (then, else, ViewModel:Device - Message View Model Object Reference)
   > n2 = Set bExecuteViewModelBindings(bExecuteViewModelBindings:bool=true) -> (Output_Get:bool)
   n3 = Get View From User Widget!(
    self:Viewmodel Engine Subsytem Object Reference=Get MVVMSubsystem() -> (ReturnValue:Viewmodel Engine Subsytem Object Reference) {is=title K2Node_GetEngineSubsystem}
    UserWidget:User Widget Object Reference=Get LoadedWidget()
   ) -> (ReturnValue:MVVM View Object Reference) {is=title K2Node_CallFunction}
   SetViewModel!(
    execute=[@n2.then, @n6.then]
    self:MVVM View Object Reference=@n3
    ViewModelName:Name=MVVM_UEFN_HUDMessage
    ViewModel:Notify Field Value Changed Interface=Get ViewModel() -> (ViewModel:Device - Message View Model Object Reference)
   ) -> (then, ReturnValue:bool)
   n4 = Get Current Details()
   > Update Viewmodel!(
    execute
    self:Device - Message View Model Object Reference=Get ViewModel()
    Details:"Gameplay Event HUDMessage Details Structure (by ref)"=@n4
    Text Style:Data Table Object Reference=GetTextStyle!(GameplayEvent_HUDMessageDetails:"Gameplay Event HUDMessage Details Structure (by ref)"=@n4) -> (Text Style Set Override:Data Table Object Reference) {is=title K2Node_CallFunction}
   ) -> (then)
   > Branch(Condition=Get bExecuteViewModelBindings() -> (bExecuteViewModelBindings:bool))
   > ExecuteViewModelBindings!(execute, self:MVVM View Object Reference=@n3, ViewModelName:Name=MVVM_UEFN_HUDMessage) -> (then, ReturnValue:bool)
   n5 = Construct Object from Class(execute=@n1.else, Class=&/MVVM_UEFN_HUDMessage.MVVM_UEFN_HUDMessage_C) -> (ReturnValue:Device - Message View Model Object Reference)
   > n6 = Set ViewModel(ViewModel:Device - Message View Model Object Reference=@n5) -> (Output_Get:Device - Message View Model Object Reference)
  ]}
 UpdateTextInputWidth {nodes [
   n1 = UpdateTextInputWidth!() -> (then, ScreenWidth:double=~, Anchor:EUIAnchor Enum=NewEnumerator0) {is=title K2Node_FunctionEntry}
   > n2 = Switch on EUIAnchor!(execute, Selection:EUIAnchor Enum=@n1.Anchor) -> (
    NewEnumerator0:Exec
    NewEnumerator1:Exec
    NewEnumerator2:Exec
    NewEnumerator3:Exec
    NewEnumerator4:Exec
    NewEnumerator5:Exec
    NewEnumerator6:Exec
    NewEnumerator7:Exec
    NewEnumerator8:Exec
    NewEnumerator9:Exec
    NewEnumerator10:Exec
    NewEnumerator11:Exec
    NewEnumerator12:Exec
    NewEnumerator13:Exec
    NewEnumerator14:Exec
    NewEnumerator15:Exec
   ) {is=title K2Node_SwitchEnum}
   n3 = Get ScreenWidth() -> (ScreenWidth:double)
   Set NewWrapTextAt(execute=[@n2.NewEnumerator13, @n2.NewEnumerator14, @n2.NewEnumerator15, @n2.NewEnumerator9], NewWrapTextAt:double=@n3) -> (Output_Get:double)
   > n4 = Set MinDesiredSize(MinDesiredSize:double=@n3) -> (Output_Get:double)
   n5 = Get ViewModel()
   Is Valid(exec=[@n6.then, @n4.then], InputObject=@n5)
   > UpdateTextInputWidth!(
    execute
    self:Device - Message View Model Object Reference=@n5
    NewWrapTextAt:double=Get NewWrapTextAt() -> (NewWrapTextAt:double)
    NewMinDesiredSize:double=Get MinDesiredSize() -> (MinDesiredSize:double)
   ) -> (then) {is=title K2Node_CallFunction}
   Set NewWrapTextAt(
    execute=[
     @n2.NewEnumerator0
     @n2.NewEnumerator1
     @n2.NewEnumerator10
     @n2.NewEnumerator11
     @n2.NewEnumerator12
     @n2.NewEnumerator2
     @n2.NewEnumerator3
     @n2.NewEnumerator4
     @n2.NewEnumerator5
     @n2.NewEnumerator6
     @n2.NewEnumerator7
     @n2.NewEnumerator8
    ]
    NewWrapTextAt=float / float(A=Get ScreenWidth(), B=3)
   )
   > n6 = Set MinDesiredSize(MinDesiredSize=0)
  ]}
 WidgetComplete {nodes [WidgetComplete!() -> (then, Reference to Self:HUDMessage Animation Container Object Reference) {is=title K2Node_FunctionEntry}]}
]
variables [
 Current Details {
  type=GameplayEvent_HUDMessageDetails
  default {
   anchor=Top Left
   audio=null
   displayTime=0
   hUD Layer=0
   icon=null
   introAnimation=None
   outroAnimation=None
   styleOverride=""
   text Justification=Left
   text Style Set Override=null
   widget Class=null
   backgroundColor=<rgba 0 0 0 0>
   message {charLimit=0, text="", message {defaultText="", key="", substitutions []}}
   offset=<0 0>
   overrideDefaultStyle {outlineStrength=0, shouldOverride=false, size=0, color=<rgba 0 0 0 0>, shadowOffset=<0 0>}
  }
 }
 DesiredVisibility {type=boolean}
 IsLoading {type=boolean}
 LoadedWidget {type=UserWidget, default=null}
 NeedsRecheck {type=boolean}
 OutroTimerHandle {type=TimerHandle}
 SelfAsCanvasPanelSlot {type=CanvasPanelSlot, default=null, class=~, properties=~}
 ViewModel {type=MVVM_UEFN_HUDMessage_C, default=null}
 WidgetComplete {}
]
widgets {
 info {inheritedWidgetCount=0, parentClass=&UserWidget, rootWidgetClass=&CanvasPanel, widgetCount=1}
 widgets [ParentPanel {widgetClassPath=&CanvasPanel, bIsVariable=true, widget=ParentPanel, properties {visibility=HitTestInvisible}}]
}
