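# HUD element widget that hosts HUD message widgets on a canvas panel, one slot per HUD layer.
# On construct it gets the playspace event router for the owning Fort player controller, listens for HUDMessage details and clear events, then broadcasts a Ready event.
# Details events update the layer's existing animation container or create one, store it by layer index and add it to the panel.
# Clear events hide that layer, optionally instantly.
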WidgetBlueprint /CreativeCoreDevices/HUDMessage/HUDMessage_Container
graphs [
 Create or Update Widget {nodes [
   n1 = Create or Update Widget!() -> (then, HUDMessageDetails:Gameplay Event HUDMessage Details Structure) {is=title K2Node_FunctionEntry}
   > Set Details(Details:Gameplay Event HUDMessage Details Structure=@n1) -> (Output_Get:Gameplay Event HUDMessage Details Structure)
   > Set HUDLayer(HUDLayer:int=Break Gameplay Event HUDMessage Details!(GameplayEvent_HUDMessageDetails:"Gameplay Event HUDMessage Details Structure (by ref)"=@n1) -> (
     HUDLayer_9_#1:int
     WidgetClass_6_#2:User Widget Soft Class Reference
     Message_72_#3:Localizable Payload Structure
     Anchor_12_#4:EUIAnchor Enum
     Offset_15_#5:Vector 2D Structure
     TextJustification_75_#6:ETextJustify Enum
     Icon_18_#7:Texture 2D Soft Object Reference
     Audio_31_#8:Sound Base Soft Object Reference
     DisplayTime_37_#9:double=~
     IntroAnimation_62_#10:EHUDMessageAnimation Enum
     OutroAnimation_65_#11:EHUDMessageAnimation Enum
     TextStyleSetOverride_45_#12:Data Table Object Reference
     OverrideDefaultStyle_58_#13:HUDMessage Rich Text Overrides Structure
     BackgroundColor_61_#14:Linear Color Structure
     StyleOverride_78_#15:String
    ) {is=title K2Node_BreakStruct}) -> (Output_Get:int=0)
   > n2 = Branch(Condition=Is Valid Index!(
     TargetArray:Array of HUDMessage Animation Container Object References=Get Message Widgets() -> (Message Widgets:Array of HUDMessage Animation Container Object References)
     IndexToTest:int=Get HUDLayer() -> (HUDLayer:int=0)
    ) -> (ReturnValue:bool))
   > Set FoundWidget(FoundWidget:HUDMessage Animation Container Object Reference="Get (a copy)"(Array:Array of HUDMessage Animation Container Object References=Get Message Widgets(), Dimension 1=Get HUDLayer() -> (HUDLayer=0)) -> (Output:HUDMessage Animation Container Object Reference)) -> (Output_Get:HUDMessage Animation Container Object Reference)
   > n3 = Get!(execute:Exec) -> (then:Exec, else:Exec, FoundWidget:HUDMessage Animation Container Object Reference)
   > UpdateFromContainer!(
    execute
    self:HUDMessage Animation Container Object Reference=@n3
    New Details:Gameplay Event HUDMessage Details Structure=Get Details() -> (Details:Gameplay Event HUDMessage Details Structure)
   ) -> (then)
   n4 = Create Widget!(
    execute:Exec=[@n2.else, @n3.else]
    Class:User Widget Class Reference=Get AnimationContainerClass() -> (AnimationContainerClass:User Widget Class Reference)
    OwningPlayer:Player Controller Object Reference
   ) -> (then:Exec, ReturnValue:User Widget Object Reference)
   > n5 = Cast To HUDMessage_AnimationContainer!(execute, Object=@n4) -> (then, CastFailed, AsHUDMessage Animation Container:HUDMessage Animation Container Object Reference) {is=title K2Node_DynamicCast}
   > Set Array Elem!(
    execute
    TargetArray:Array of HUDMessage Animation Container Object References=Get Message Widgets()
    Index:int=Get HUDLayer() -> (HUDLayer=0)
    Item:"HUDMessage Animation Container Object Reference (by ref)"=@n5
    bSizeToFit:bool=true
   ) -> (then)
   > AddChildToCanvas!(execute, self:Canvas Panel Object Reference=Get Panel() -> (Panel:Canvas Panel Object Reference), Content:Widget Object Reference=@n5) -> (then, ReturnValue:Canvas Panel Slot Object Reference)
   > UpdateFromContainer(self=@n5, New Details=Get Details())
  ]}
 EventGraph {nodes [
   n1 = On HUD Message Clear Received!() -> (OutputDelegate, then, Payload_HUDLayer_5_#16:int, Payload_Immediate_2_#17:bool=~) {is=title K2Node_CustomEvent}
   > UpdateVisibilityForLayer!(
    execute
    self
    Layer:int=@n1.Payload_HUDLayer_5_#16
    On:bool=false
    Instant:bool=@n1.Payload_Immediate_2_#17
   ) -> (then)
   Initialize Event Router Message Listeners() {is=title K2Node_CustomEvent}
   > n2 = GetEventRouter!(
    execute
    Scope:Object Class Reference=/Script/FortniteGame.FortPlayspace
    ContextActor:Actor Object Reference=Get CachedPlayerController() -> (CachedPlayerController:Fort Player Controller Object Reference)
   ) -> (then, ReturnValue:Gameplay Event Router Component Object Reference)
   > n3 = Set Event Router(Event Router:Gameplay Event Router Component Object Reference=@n2) -> (Output_Get:Gameplay Event Router Component Object Reference)
   n4 = Get CachedPlayerController()
   > n5 = StartListeningToEvent!(
    execute:Exec
    Target:Gameplay Event Router Component Object Reference=@n3
    EventType:Script Struct Object Reference=&/GameplayEvent_HUDMessageClear
    Context:Object Reference=@n4
    EventBubblingRule:EEventBubblingRule Enum
    None:%=@n4
   ) -> (
    then:Exec
    AsyncTaskProxy:Async Action Start Listening To Event Object Reference
    OnEventReceived:Exec
    Payload:Gameplay Event HUDMessage Clear Structure
   )
   n6 = Get CachedPlayerController()
   > n7 = StartListeningToEvent(
    Target=Get Event Router() -> (Event Router:Gameplay Event Router Component Object Reference)
    EventType=&/GameplayEvent_HUDMessageDetails
    Context=@n6
    None=@n6
   ) -> (Payload:Gameplay Event HUDMessage Details Structure)
   > Broadcast Event!(
    execute
    self:Gameplay Event Router Component Object Reference=Get Event Router()
    EventData:Gameplay Event HUDMessage Ready Structure=Make Gameplay Event HUDMessage Ready!(NullVar_6_#17:bool=False) -> (GameplayEvent_HUDMessageReady:Gameplay Event HUDMessage Ready Structure) {is=title K2Node_MakeStruct}
    EventContext:Object Reference=Get CachedPlayerController()
   ) -> (then, ReturnValue:bool)
   Create or Update Widget!(execute=@n7.OnEventReceived, self, HUDMessageDetails:Gameplay Event HUDMessage Details Structure=@n7.Payload) -> (then) {is=title K2Node_CallFunction}
   On HUD Message Clear Received!(execute=@n5.OnEventReceived, self, Payload:Gameplay Event HUDMessage Clear Structure=@n5.Payload) -> (then) {is=title K2Node_CallFunction}
   Event Construct()
   > n8 = Cast To FortPlayerController(Object=GetOwningPlayer())
   > Set CachedPlayerController(CachedPlayerController:Fort Player Controller Object Reference=@n8) -> (Output_Get:Fort Player Controller Object Reference)
   > Initialize Event Router Message Listeners() {is=title K2Node_CallFunction}
  ]}
 UpdateVisibilityForLayer {nodes [
   n1 = UpdateVisibilityForLayer!() -> (then, Layer:int, On:bool=~, Instant:bool=~) {is=title K2Node_FunctionEntry}
   n2 = Get Message Widgets()
   > Branch(Condition=Is Valid Index(TargetArray=@n2, IndexToTest=@n1))
   > Set FoundWidget(FoundWidget="Get (a copy)"(Array=@n2, Dimension 1=@n1))
   > n3 = Get()
   > UpdateMessageVisibility!(execute, self:HUDMessage Animation Container Object Reference=@n3, On:bool=@n1.On, Instant:bool=@n1.Instant) -> (then)
  ]}
]
variables [
 AnimationContainerClass {type=Class, default=&/HUDMessage_AnimationContainer.HUDMessage_AnimationContainer_C}
 CachedPlayerController {type=FortPlayerController, default=null}
 Event Router {type=GameplayEventRouterComponent, default=null}
 Message Widgets {type="array<HUDMessage_AnimationContainer_C>", default []}
]
widgets {
 info {inheritedWidgetCount=1, parentClass=&FortHUDElementWidget, rootWidgetClass=&CanvasPanel, widgetCount=2}
 widgets [
  Panel {widgetClassPath=&CanvasPanel, bIsVariable=true, widget=Panel}
  TutorialHighlight {widgetClassPath=&TutorialHighlight}
 ]
}
