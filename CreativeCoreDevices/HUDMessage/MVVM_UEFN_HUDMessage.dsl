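; MVVM viewmodel for the UEFN HUD Message device, holding the message text, icon, background and rich-text styling to display.
; Update Viewmodel unpacks the message details struct and text style table into its variables, while UpdateTextInputWidth sets wrap and minimum width.
; GetDefaultStyleOverrides merges stored font size, outline, color and shadow offset into a given text block style.
; SetBackgroundDetails hides the background unless its alpha is above zero.
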(event Custom|Update%20Viewmodel
  (params (Details :type Gameplay%20Event%20HUDMessage%20Details%20Structure%20%28by%20ref%29) (Text%20Style :type Data%20Table%20Object%20Reference))
  (bind (_shouldoverride_1_0c75cc1a4e5e63d5d5969e87bb1d63a2 _color_4_e49f89984cf19e4435e523950c01f36c _size_7_54467cc34710d576b31ee2aac68e44e9 _outlinestrength_10_2ccb31034eb903984a7cf68f2553b983 _shadowoffset_15_4bf8aabf4661082f340176ba4a7a1f92) (Utilities|Struct|BreakHUDMessageRichTextOverrides (.OverrideDefaultStyle_58_21D237ED4CCA3EA851153E9FA178C325 (Utilities|Struct|BreakGameplayEventHUDMessageDetails Details))))
  (bind (_message_72_7634162b4c068206f407bcac0c836150 _textjustification_75_68386cab41ed610c8696fa80e715b116 _icon_18_9a02ea8d45712d1efb554fb3806b0755 _overridedefaultstyle_58_21d237ed4cca3ea851153e9fa178c325 _backgroundcolor_61_7b205098499357f044a3cbaea74542ec) (Utilities|Struct|BreakGameplayEventHUDMessageDetails Details))
  (bind _message_72_7634162b4c068206f407bcac0c836150_text_2_563e0da1490d74cee84ef39a5df6bb26 (Utilities|Struct|BreakLocalizablePayload _message_72_7634162b4c068206f407bcac0c836150))
  (Variables|Default|SetText%20Justification _textjustification_75_68386cab41ed610c8696fa80e715b116)
  (Variables|Default|SetText _message_72_7634162b4c068206f407bcac0c836150_text_2_563e0da1490d74cee84ef39a5df6bb26)
  (CallFunction|SetBackgroundDetails _backgroundcolor_61_7b205098499357f044a3cbaea74542ec)
  (Variables|Default|SetIcon _icon_18_9a02ea8d45712d1efb554fb3806b0755)
  (Variables|RichTextStyling|SetText%20Style%20Set Text%20Style)
  (Variables|RichTextStyling|DefaultOverrides|SetTextColor (Utilities|Struct|MakeSlateColor _color_4_e49f89984cf19e4435e523950c01f36c))
  (Variables|RichTextStyling|DefaultOverrides|SetShadowOffset _shadowoffset_15_4bf8aabf4661082f340176ba4a7a1f92)
  (Variables|RichTextStyling|DefaultOverrides|SetOutlineStrength (Math|Conversions|ToInteger64(Integer) _outlinestrength_10_2ccb31034eb903984a7cf68f2553b983))
  (Variables|RichTextStyling|DefaultOverrides|SetFontSize (Math|Conversions|ToInteger64(Integer) _size_7_54467cc34710d576b31ee2aac68e44e9))
  (Variables|RichTextStyling|DefaultOverrides|SetOverrideDefaultStyle _shouldoverride_1_0c75cc1a4e5e63d5d5969e87bb1d63a2))

(event Custom|UpdateTextInputWidth
  (params (NewWrapTextAt :type Float%20%28double-precision%29) (NewMinDesiredSize :type Float%20%28double-precision%29))
  (Variables|Default|SetWrapTextAt NewWrapTextAt)
  (Variables|Default|SetMinDesiredSize NewMinDesiredSize))

(fn GetDefaultStyleOverrides
  (params (Override :type Text%20Block%20Style%20Structure))
  (bind _font (Utilities|Struct|BreakTextBlockStyle Override))
  ;[0 0.787 1 0.2] Override the values of the Default Text Style with the values defined by the Default Style Override (if set). else clear the default style override.
  (bind _structout (Utilities|Struct|SetmembersinFontOutlineSettings (Utilities|Struct|BreakSlateFontInfo _font) (Math|Conversions|ToInteger(Integer64) (Variables|RichTextStyling|DefaultOverrides|GetOutlineStrength))))
  (bind _structout_2 (Utilities|Struct|SetmembersinSlateFontInfo _font _structout (Math|Conversions|ToFloat(Integer64) (Variables|RichTextStyling|DefaultOverrides|GetFontSize))))
  (bind _structout_3 (Utilities|Struct|SetmembersinTextBlockStyle Override _structout_2 (Variables|RichTextStyling|DefaultOverrides|GetTextColor) (Variables|RichTextStyling|DefaultOverrides|GetShadowOffset)))
  (return _structout_3))

(fn SetBackgroundDetails
  (params (SpecifiedColor :type Linear%20Color%20Structure))
  (bind (_r _g _b _a) (Utilities|Struct|BreakLinearColor SpecifiedColor))
  (Variables|Default|SetBackgroundColor (Utilities|Struct|MakeSlateColor SpecifiedColor))
  (Variables|Default|SetBackgroundVisibility (select (> _a 0) "SelfHitTestInvisible" "Hidden")))
