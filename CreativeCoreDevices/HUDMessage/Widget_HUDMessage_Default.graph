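# Default UMG widget for the Creative HUD Message device, showing centered rich text with a collapsed-by-default icon slot over a tinted background border.
# MVVM-named override functions for color, font size, outline, shadow and style call Update Style, which applies the device view model's overrides or clears them.
# A visibility function switches text padding to match the background.
# Intro and Outro animations scale and fade the root canvas; the event graph is empty.
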WidgetBlueprint /CreativeCoreDevices/HUDMessage/Widget_HUDMessage_Default
animations [
 Intro {
  binding_count=1
  display_label=""
  length_seconds=0.25
  object=&self:Intro
  sequence {bindings [CanvasPanel_1 {
     id=#1
     possessed_class=CanvasPanel
     tracks [
      {
       type=MovieSceneFloatTrack
       display_name=Render Opacity
       sections [{type=MovieSceneFloatSection, end_frame=4, start_frame=0, channels [None {type=MovieSceneScriptingFloatChannel, keys [[0,0],[2.7623,1]]}]}]
      }
      {
       type=MovieScene2DTransformTrack
       sections [{
         type=MovieScene2DTransformSection
         end_frame=5
         start_frame=0
         channels [
          Translation.X {type=MovieSceneScriptingFloatChannel}
          Translation.Y {type=MovieSceneScriptingFloatChannel}
          Angle {type=MovieSceneScriptingFloatChannel}
          Scale.X {type=MovieSceneScriptingFloatChannel, keys [[0,1],[3,1.282],[3.5983,1],[4.2237,1]]}
          Scale.Y {type=MovieSceneScriptingFloatChannel, keys [[0,0],[3,1.282],[3.5983,1],[4.2237,1]]}
          Shear.X {type=MovieSceneScriptingFloatChannel, const [0,1]}
          Shear.Y {type=MovieSceneScriptingFloatChannel, const [0,1]}
         ]
        }]
      }
     ]
    }]}
 }
 Outro {
  binding_count=1
  display_label=""
  length_seconds=0.3397
  object=&self:Outro
  sequence {
   bindings [CanvasPanel_1 {
     id=#2
     possessed_class=CanvasPanel
     tracks [
      {
       type=MovieScene2DTransformTrack
       sections [{
         type=MovieScene2DTransformSection
         end_frame=40
         start_frame=0
         channels [
          Translation.X {type=MovieSceneScriptingFloatChannel}
          Translation.Y {type=MovieSceneScriptingFloatChannel}
          Angle {type=MovieSceneScriptingFloatChannel}
          Scale.X {type=MovieSceneScriptingFloatChannel, const [1,2]}
          Scale.Y {type=MovieSceneScriptingFloatChannel, keys [[0,1],[9,1.2783],[11,1],[40.756,0]]}
          Shear.X {type=MovieSceneScriptingFloatChannel}
          Shear.Y {type=MovieSceneScriptingFloatChannel}
         ]
        }]
      }
      {
       type=MovieSceneFloatTrack
       display_name=Render Opacity
       sections [{type=MovieSceneFloatSection, end_frame=40, start_frame=0, channels [None {type=MovieSceneScriptingFloatChannel, keys [[0,1],[40.756,0]]}]}]
      }
     ]
    }]
   display_rate {numerator=120}
  }
 }
]
graphs [
 EventGraph {links [], nodes []}
 MVVM OverrideBackgroundVisibility {nodes [
   n1 = MVVM OverrideBackgroundVisibility!() -> (then, InVisibility:ESlateVisibility Enum=Visible) {is=title K2Node_FunctionEntry}
   n2 = Get BackgroundMargin() -> (BackgroundMargin:Margin Structure)
   n3 = Get NoMargin() -> (NoMargin:Margin Structure)
   > SetPadding!(
    execute
    self:Vertical Box Slot Object Reference=SlotAsVerticalBoxSlot(Widget=Get MessageTextBlock() -> (MessageTextBlock:Rich Text Block Object Reference))
    InPadding:Margin Structure=Select!(
     Visible:Margin Structure=@n2
     Collapsed:Margin Structure=@n3
     Hidden:Margin Structure=@n3
     HitTestInvisible:Margin Structure=@n2
     SelfHitTestInvisible:Margin Structure=@n2
     Index:ESlateVisibility Enum=@n1
    ) -> (ReturnValue:Margin Structure)
   ) -> (then)
  ]}
 MVVM OverrideDefaultStyle {nodes [
   MVVM OverrideDefaultStyle!() -> (then, Override:bool=~) {is=title K2Node_FunctionEntry}
   > Update Style()
   Break Text Block Style!(TextBlockStyle:"Text Block Style Structure (by ref)") -> (
    Font:Slate Font Info Structure
    ColorAndOpacity:Slate Color Structure
    ShadowOffset:Vector 2D Structure
    ShadowColorAndOpacity:Linear Color Structure
    HighlightColor:Slate Color Structure
    HighlightShape:Slate Brush Structure
    StrikeBrush:Slate Brush Structure
    UnderlineBrush:Slate Brush Structure
    TransformPolicy:ETextTransformPolicy Enum
    OverflowPolicy:ETextOverflowPolicy Enum
   ) {is=title K2Node_BreakStruct}
  ]}
 MVVM OverrideFontSize {nodes [
   MVVM OverrideFontSize!() -> (then, Override:Integer64) {is=title K2Node_FunctionEntry}
   > Update Style()
  ]}
 MVVM OverrideOutlineStrength {nodes [
   MVVM OverrideOutlineStrength!() -> (then, Override:Integer64) {is=title K2Node_FunctionEntry}
   > Update Style()
  ]}
 MVVM OverrideShadowOffset {nodes [
   MVVM OverrideShadowOffset!() -> (then, Override:Vector 2D Structure) {is=title K2Node_FunctionEntry}
   > Update Style()
  ]}
 MVVM OverrideTextColor {nodes [
   MVVM OverrideTextColor!() -> (then, Override:Slate Color Structure) {is=title K2Node_FunctionEntry}
   > Update Style()
  ]}
 Update Style {nodes [
   Update Style() {is=title K2Node_FunctionEntry}
   > n1 = Get!(execute:Exec) -> (then:Exec, else:Exec, MVVM_UEFN_HUDMessage:Device - Message View Model Object Reference)
   > n2 = Branch(Condition=Get OverrideDefaultStyle!(self:Device - Message View Model Object Reference=@n1) -> (OverrideDefaultStyle:bool))
   n3 = Get MessageTextBlock()
   > n4 = Set DefaultStyle(DefaultStyle:Text Block Style Structure=Get DefaultTextStyleOverride!(self:Rich Text Block Object Reference=@n3) -> (DefaultTextStyleOverride:Text Block Style Structure)) -> (Output_Get:Text Block Style Structure)
   > n5 = Get Default Style Overrides!(execute, self:Device - Message View Model Object Reference=@n1, Override:Text Block Style Structure=@n4) -> (then, StructOut:Text Block Style Structure) {is=title K2Node_CallFunction}
   > SetDefaultTextStyle!(execute, self:Rich Text Block Object Reference=Get MessageTextBlock(), InDefaultTextStyle:"Text Block Style Structure (by ref)"=@n5) -> (then)
   ClearAllDefaultStyleOverrides(execute=@n2.else, self:Rich Text Block Object Reference=@n3)
  ]}
]
variables [
 BackgroundMargin {type=Margin, default=<ltrb 10 10 10 10>}
 NoMargin {type=Margin, default=<ltrb 0 0 0 0>}
 OutroTimerHandle {type=TimerHandle}
 Owning Device {type=Device_HUDMessage_V2_C, default=null}
 Owning Player Component {type=CRD_HUDMessageComp_C, default=null}
 ShowTimeTimerHandle {type=TimerHandle}
]
widgets {
 info {inheritedWidgetCount=0, parentClass=&UserWidget, rootWidgetClass=&CanvasPanel, widgetCount=5}
 widgets [
  CanvasPanel_1 {widgetClassPath=&CanvasPanel, widget=CanvasPanel_1}
  BackgroundBorder {
   widgetClassPath=&CommonBorder
   parent=CanvasPanel_1
   widget=BackgroundBorder
   properties {
    background {
     bIsDynamicallyLoaded=false
     drawAs=Box
     imageType=NoImage
     mirroring=NoMirror
     resourceName=None
     resourceObject=&/M_HUDMessage_BasicBackground
     tiling=NoTile
     imageSize=<64 64>
     margin=<ltrb 0.5 0.5 0.5 0.5>
     outlineSettings {
      bUseBrushTransparency=false
      roundingType=HalfHeightRadius
      width=0
      color {colorUseRule=UseColor_Specified, specifiedColor=<rgba 0 0 0 0>}
      cornerRadii=<0 0 0 0>
     }
     tintColor {colorUseRule=UseColor_Specified, specifiedColor=<rgba 0 0.098 1 0.5>}
     uVRegion {bIsValid=false, max=<0 0>, min=<0 0>}
    }
    padding=<ltrb 0 0 0 0>
   }
   slot {class=CanvasPanelSlot, properties {layoutData {alignment=<0.5 0.5>, anchors {maximum=<1 1>, minimum=<0 0>}, offsets=<ltrb 0 0 0 0>}}}
  }
  MessageBox {
   widgetClassPath=&VerticalBox
   bIsVariable=true
   parent=CanvasPanel_1
   widget=MessageBox
   slot {
    class=CanvasPanelSlot
    properties {bAutoSize=true, layoutData {alignment=<0.5 0.5>, anchors {maximum=<0.5 0.5>, minimum=<0.5 0.5>}, offsets=<ltrb 0 0 0 0>}}
   }
  }
  Icon {
   widgetClassPath=&Image
   bIsVariable=true
   parent=MessageBox
   widget=Icon
   properties {
    visibility=Collapsed
    brush {
     bIsDynamicallyLoaded=false
     drawAs=Image
     imageType=NoImage
     mirroring=NoMirror
     resourceName=None
     resourceObject=null
     tiling=NoTile
     imageSize=<64 64>
     margin=<ltrb 0 0 0 0>
     outlineSettings {
      bUseBrushTransparency=false
      roundingType=HalfHeightRadius
      width=0
      color {colorUseRule=UseColor_Specified, specifiedColor=<rgba 0 0 0 0>}
      cornerRadii=<0 0 0 0>
     }
     tintColor {colorUseRule=UseColor_Specified, specifiedColor=<rgba 1 1 1 1>}
     uVRegion {bIsValid=false, max=<0 0>, min=<0 0>}
    }
   }
   slot {class=VerticalBoxSlot, properties {horizontalAlignment=HAlign_Center, verticalAlignment=VAlign_Bottom, padding=<ltrb 0 0 0 10>}}
  }
  MessageTextBlock {
   widgetClassPath=/Script/UMG.RichTextBlock
   bIsVariable=true
   parent=MessageBox
   widget=MessageTextBlock
   properties {
    justification=Center
    lineHeightPercentage=0.85
    text="<emphasis>HUD Message</>"
    textStyleSet=/Game/Athena/UI/InGame/Creative/PropertyEditor/StyleEditor/RichTextStyleSets/CreativeDefaultRichTextStyles
    decoratorClasses [/Game/Athena/HUD/TextStyles/AthenaScoreAlertDecorator.AthenaScoreAlertDecorator_C]
    defaultTextStyleOverride {
     overflowPolicy=Clip
     transformPolicy=None
     colorAndOpacity {colorUseRule=UseColor_Specified, specifiedColor=<rgba 1 1 1 1>}
     font {
      bForceMonospaced=false
      bMaterialIsStencil=false
      fontMaterial=null
      fontObject=&BurbankSmall
      letterSpacing=0
      monospacedWidth=1
      size=18
      skewAmount=0
      typefaceFontName=Bold
      outlineSettings {
       bApplyOutlineToDropShadows=true
       bMiteredCorners=false
       bSeparateFillAlpha=false
       outlineMaterial=null
       outlineSize=0
       outlineColor=<rgba 0 0 0 1>
      }
     }
     highlightColor {colorUseRule=UseColor_Specified, specifiedColor=<rgba 0 0 0 1>}
     highlightShape {
      bIsDynamicallyLoaded=false
      drawAs=Image
      imageType=NoImage
      mirroring=NoMirror
      resourceName=None
      resourceObject=null
      tiling=NoTile
      imageSize=<32 32>
      margin=<ltrb 0 0 0 0>
      outlineSettings {
       bUseBrushTransparency=false
       roundingType=HalfHeightRadius
       width=0
       color {colorUseRule=UseColor_Specified, specifiedColor=<rgba 0 0 0 0>}
       cornerRadii=<0 0 0 0>
      }
      tintColor {colorUseRule=UseColor_Specified, specifiedColor=<rgba 1 1 1 1>}
      uVRegion {bIsValid=false, max=<0 0>, min=<0 0>}
     }
     selectedBackgroundColor {colorUseRule=UseColor_Foreground, specifiedColor=<rgba 1 0 1 1>}
     shadowColorAndOpacity=<rgba 0 0 0 1>
     shadowOffset=<0 0>
     strikeBrush {
      bIsDynamicallyLoaded=false
      drawAs=Image
      imageType=NoImage
      mirroring=NoMirror
      resourceName=None
      resourceObject=null
      tiling=NoTile
      imageSize=<32 32>
      margin=<ltrb 0 0 0 0>
      outlineSettings {
       bUseBrushTransparency=false
       roundingType=HalfHeightRadius
       width=0
       color {colorUseRule=UseColor_Specified, specifiedColor=<rgba 0 0 0 0>}
       cornerRadii=<0 0 0 0>
      }
      tintColor {colorUseRule=UseColor_Specified, specifiedColor=<rgba 1 1 1 1>}
      uVRegion {bIsValid=false, max=<0 0>, min=<0 0>}
     }
     underlineBrush {
      bIsDynamicallyLoaded=false
      drawAs=Image
      imageType=NoImage
      mirroring=NoMirror
      resourceName=None
      resourceObject=null
      tiling=NoTile
      imageSize=<32 32>
      margin=<ltrb 0 0 0 0>
      outlineSettings {
       bUseBrushTransparency=false
       roundingType=HalfHeightRadius
       width=0
       color {colorUseRule=UseColor_Specified, specifiedColor=<rgba 0 0 0 0>}
       cornerRadii=<0 0 0 0>
      }
      tintColor {colorUseRule=UseColor_Specified, specifiedColor=<rgba 1 1 1 1>}
      uVRegion {bIsValid=false, max=<0 0>, min=<0 0>}
     }
    }
    renderTransform {angle=0, scale=<1 1>, shear=<-4 0>, translation=<0 0>}
   }
   slot {class=VerticalBoxSlot}
  }
 ]
}
