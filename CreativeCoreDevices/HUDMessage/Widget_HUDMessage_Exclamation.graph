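# HUD Message device widget that shows rich-text messages with an animated "!" in the same overlay.
# PlayIntro recolours the "!" by looking up the message's first markup tag in a colour map, reveals the widget and plays its fade, squash-and-drop animation.
# MVVM override handlers rebuild the message's default text style from the view model or clear overrides.
# Message padding switches between margin and none with background visibility.
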WidgetBlueprint /CreativeCoreDevices/HUDMessage/Widget_HUDMessage_Exclamation
animations [Exclamation {
  binding_count=1
  display_label=""
  length_seconds=1.25
  object=&self:Exclamation
  sequence {
   bindings [ExclamationTextBlock {
     id=#1
     possessed_class=TextBlock
     tracks [
      {
       type=MovieSceneFloatTrack
       display_name=Render Opacity
       sections [{
         type=MovieSceneFloatSection
         end_frame=149
         start_frame=0
         channels [None {type=MovieSceneScriptingFloatChannel, keys [[0,0],[15.774,1],[123.702,1],[149.426,0]]}]
        }]
      }
      {
       type=MovieScene2DTransformTrack
       sections [{
         type=MovieScene2DTransformSection
         end_frame=150
         start_frame=0
         channels [
          Translation.X {type=MovieSceneScriptingFloatChannel}
          Translation.Y {type=MovieSceneScriptingFloatChannel, keys [[0,0],[19,0],[124,90],[150,130]]}
          Angle {type=MovieSceneScriptingFloatChannel}
          Scale.X {
           type=MovieSceneScriptingFloatChannel
           keys [[0,1],[19,2.1864],[29,0.7511],[36,1.201],[43,0.8469],[53.844,1],[117.032,1],[127,1.1192],[149.584,1],[149.986,0]]
          }
          Scale.Y {
           type=MovieSceneScriptingFloatChannel
           keys [[0,1],[19,2.9101],[29,0.6527],[36,1.2787],[43,0.7639],[53.844,1],[117.032,1],[127,1.3143],[146,0.918],[149.986,0]]
          }
          Shear.X {type=MovieSceneScriptingFloatChannel}
          Shear.Y {type=MovieSceneScriptingFloatChannel}
         ]
        }]
      }
     ]
    }]
   display_rate {numerator=120}
  }
 }]
graphs [
 EventGraph {nodes [
   Event PlayIntro() {is=title K2Node_Event}
   > UpdateExclamationTextBlockColors()
   > SetVisibility(InVisibility=SelfHitTestInvisible)
   > PlayAnimation(InAnimation=Get Exclamation() -> (Exclamation:Widget Animation Object Reference))
  ]}
 MVVM OverrideBackgroundVisibility {nodes [
   n1 = MVVM OverrideBackgroundVisibility!() -> (then, InVisibility:ESlateVisibility Enum=Visible) {is=title K2Node_FunctionEntry}
   n2 = Get BackgroundMargin() -> (BackgroundMargin:Margin Structure)
   n3 = Get NoMargin() -> (NoMargin:Margin Structure)
   > SetPadding!(
    execute
    self:Overlay Slot Object Reference=SlotAsOverlaySlot(Widget=Get MessageTextBlock() -> (MessageTextBlock:Rich Text Block Object Reference))
    InPadding:Margin Structure=Select!(
     Visible:Margin Structure=@n2
     Collapsed:Margin Structure=@n3
     Hidden:Margin Structure=@n3
     HitTestInvisible:Margin Structure=@n2
     SelfHitTestInvisible:Margin Structure=@n2
     Index:ESlateVisibility Enum=@n1
    ) -> (ReturnValue:Margin Structure)
   ) -> (then)
  ]}
 MVVM OverrideColor {nodes [
   MVVM OverrideColor!() -> (then, Override:Slate Color Structure) {is=title K2Node_FunctionEntry}
   > Update Style()
  ]}
 MVVM OverrideDefaultStyle {nodes [
   MVVM OverrideDefaultStyle!() -> (then, Override:bool=~) {is=title K2Node_FunctionEntry}
   > Update Style()
  ]}
 MVVM OverrideOutline {nodes [
   MVVM OverrideOutline!() -> (then, Override:Integer64) {is=title K2Node_FunctionEntry}
   > Update Style()
  ]}
 MVVM OverrideShadow {nodes [
   MVVM OverrideShadow!() -> (then, Override:Vector 2D Structure) {is=title K2Node_FunctionEntry}
   > Update Style()
  ]}
 MVVM OverrideTextSize {nodes [
   MVVM OverrideTextSize!() -> (then, Override:Integer64) {is=title K2Node_FunctionEntry}
   > Update Style()
  ]}
 Update Style {nodes [
   Update Style() {is=title K2Node_FunctionEntry}
   > n1 = Get!(execute:Exec) -> (then:Exec, else:Exec, MVVM_UEFN_HUDMessage:Device - Message View Model Object Reference)
   > n2 = Branch(Condition=Get OverrideDefaultStyle!(self:Device - Message View Model Object Reference=@n1) -> (OverrideDefaultStyle:bool))
   n3 = Get MessageTextBlock()
   > n4 = Set DefaultStyle(DefaultStyle:Text Block Style Structure=Get DefaultTextStyleOverride!(self:Rich Text Block Object Reference=@n3) -> (DefaultTextStyleOverride:Text Block Style Structure)) -> (Output_Get:Text Block Style Structure)
   > n5 = Get Default Style Overrides!(execute, self:Device - Message View Model Object Reference=@n1, Override:Text Block Style Structure=@n4) -> (then, StructOut:Text Block Style Structure) {is=title K2Node_CallFunction}
   > SetDefaultTextStyle!(execute, self:Rich Text Block Object Reference=Get MessageTextBlock(), InDefaultTextStyle:"Text Block Style Structure (by ref)"=@n5) -> (then)
   ClearAllDefaultStyleOverrides(execute=@n2.else, self:Rich Text Block Object Reference=@n3)
  ]}
 UpdateExclamationTextBlockColors {nodes [
   UpdateExclamationTextBlockColors() {is=title K2Node_FunctionEntry}
   n1 = "To String (Text)"(InText=Get Text(self:Rich Text Block Object Reference=Get MessageTextBlock()))
   n2 = FindSubstring(SearchIn=@n1, Substring=">")
   n3 = FindSubstring(SearchIn=@n1, Substring="<")
   n4 = Find(
    TargetMap:Map of Strings to Slate Color Structures=Get HUDMessageColors() -> (HUDMessageColors:Map of Strings to Slate Color Structures)
    Key=GetSubstring!(
     SourceString:String=@n1
     StartIndex:int=int + int(A=@n3, B=1) -> (ReturnValue=0) {is=title K2Node_CommutativeAssociativeBinaryOperator}
     Length:int=int - int(A=@n2, B=1) {is=title K2Node_CallFunction}
    ) -> (ReturnValue:String)
   ) -> (Value:Slate Color Structure)
   > Branch(Condition=@n4.ReturnValue)
   > SetColorAndOpacity(self=Get ExclamationTextBlock() -> (ExclamationTextBlock:Text Object Reference), InColorAndOpacity=@n4)
  ]}
]
variables [
 BackgroundMargin {type=Margin, default=<ltrb 10 10 10 10>}
 HUDMessageColors {
  type="map<SlateColor>"
  default {
   Ally {colorUseRule=UseColor_Specified, specifiedColor=<rgba 0 0.4377 0.8958 1>}
   Enemy {colorUseRule=UseColor_Specified, specifiedColor=<rgba 0.6875 0.0809 0 1>}
   Heist {colorUseRule=UseColor_Specified, specifiedColor=<rgba 1 1 1 1>}
   HeistEnemy {colorUseRule=UseColor_Specified, specifiedColor=<rgba 0.6875 0.0809 0 1>}
   Intro {colorUseRule=UseColor_Specified, specifiedColor=<rgba 0 0.491 1 1>}
   IntroScore {colorUseRule=UseColor_Specified, specifiedColor=<rgba 1 1 1 1>}
   RespawnAlert {colorUseRule=UseColor_Specified, specifiedColor=<rgba 0.9883 1 0.8741 1>}
   RespawnAlertB {colorUseRule=UseColor_Specified, specifiedColor=<rgba 1 0.0247 0 1>}
   RespawnAlertBLrg {colorUseRule=UseColor_Specified, specifiedColor=<rgba 1 1 1 1>}
   Score {colorUseRule=UseColor_Specified, specifiedColor=<rgba 1 1 1 1>}
  }
 }
 NoMargin {type=Margin, default=<ltrb 0 0 0 0>}
 OutroTimerHandle {type=TimerHandle}
 Owning Device {type=Device_HUDMessage_V2_C, default=null}
 Owning Player Component {type=CRD_HUDMessageComp_C, default=null}
 ShowTimeTimerHandle {type=TimerHandle}
]
widgets {
 info {inheritedWidgetCount=0, parentClass=&UserWidget, rootWidgetClass=&CanvasPanel, widgetCount=5}
 widgets [
  CanvasPanel_1 {widgetClassPath=&CanvasPanel, widget=CanvasPanel_1}
  BackgroundBorder {
   widgetClassPath=&CommonBorder
   parent=CanvasPanel_1
   widget=BackgroundBorder
   properties {
    background {
     bIsDynamicallyLoaded=false
     drawAs=Box
     imageType=NoImage
     mirroring=NoMirror
     resourceName=None
     resourceObject=&/M_HUDMessage_BasicBackground
     tiling=NoTile
     imageSize=<64 64>
     margin=<ltrb 0.5 0.5 0.5 0.5>
     outlineSettings {
      bUseBrushTransparency=false
      roundingType=HalfHeightRadius
      width=0
      color {colorUseRule=UseColor_Specified, specifiedColor=<rgba 0 0 0 0>}
      cornerRadii=<0 0 0 0>
     }
     tintColor {colorUseRule=UseColor_Specified, specifiedColor=<rgba 0 0.098 1 0.5>}
     uVRegion {bIsValid=false, max=<0 0>, min=<0 0>}
    }
    padding=<ltrb 0 0 0 0>
   }
   slot {class=CanvasPanelSlot, properties {layoutData {alignment=<0.5 0.5>, anchors {maximum=<1 1>, minimum=<0 0>}, offsets=<ltrb 0 0 0 0>}}}
  }
  OverlayFeedback {
   widgetClassPath=&Overlay
   bIsVariable=true
   parent=CanvasPanel_1
   widget=OverlayFeedback
   slot {
    class=CanvasPanelSlot
    properties {bAutoSize=true, layoutData {alignment=<0.5 0.5>, anchors {maximum=<0.5 0.5>, minimum=<0.5 0.5>}, offsets=<ltrb 0 0 0 0>}}
   }
  }
  ExclamationTextBlock {
   widgetClassPath=&TextBlock
   bIsVariable=true
   parent=OverlayFeedback
   widget=ExclamationTextBlock
   properties {
    renderOpacity=0
    text="!"
    colorAndOpacity {colorUseRule=UseColor_Specified, specifiedColor=<rgba 1 0.679 0.3121 1>}
    font {
     bForceMonospaced=false
     bMaterialIsStencil=false
     fontMaterial=/FortUILibrary/Materials/Foundation/M_Font_Horizon_Dynamic
     fontObject=&Burbank
     letterSpacing=0
     monospacedWidth=1
     size=72
     skewAmount=0
     typefaceFontName=Burbank-Black
     outlineSettings {
      bApplyOutlineToDropShadows=false
      bMiteredCorners=false
      bSeparateFillAlpha=false
      outlineMaterial=null
      outlineSize=0
      outlineColor=<rgba 0 0 0 1>
     }
    }
    renderTransformPivot=<0.5 0.7834>
   }
   slot {class=OverlaySlot, properties {horizontalAlignment=HAlign_Center, padding=<ltrb -200 -200 -200 -200>}}
  }
  MessageTextBlock {
   widgetClassPath=/Script/UMG.RichTextBlock
   bIsVariable=true
   parent=OverlayFeedback
   widget=MessageTextBlock
   properties {
    justification=Center
    lineHeightPercentage=0.85
    text="<emphasis>HUD Message</><warning>!</>"
    textStyleSet=/Game/Athena/UI/InGame/Creative/PropertyEditor/StyleEditor/RichTextStyleSets/CreativeDefaultRichTextStyles
    decoratorClasses [/Game/Athena/HUD/TextStyles/AthenaScoreAlertDecorator.AthenaScoreAlertDecorator_C]
    defaultTextStyleOverride {
     overflowPolicy=Clip
     transformPolicy=None
     colorAndOpacity {colorUseRule=UseColor_Specified, specifiedColor=<rgba 1 1 1 1>}
     font {
      bForceMonospaced=false
      bMaterialIsStencil=false
      fontMaterial=null
      fontObject=&BurbankSmall
      letterSpacing=0
      monospacedWidth=1
      size=18
      skewAmount=0
      typefaceFontName=Bold
      outlineSettings {
       bApplyOutlineToDropShadows=true
       bMiteredCorners=false
       bSeparateFillAlpha=false
       outlineMaterial=null
       outlineSize=0
       outlineColor=<rgba 0 0 0 1>
      }
     }
     highlightColor {colorUseRule=UseColor_Specified, specifiedColor=<rgba 0 0 0 1>}
     highlightShape {
      bIsDynamicallyLoaded=false
      drawAs=Image
      imageType=NoImage
      mirroring=NoMirror
      resourceName=None
      resourceObject=null
      tiling=NoTile
      imageSize=<32 32>
      margin=<ltrb 0 0 0 0>
      outlineSettings {
       bUseBrushTransparency=false
       roundingType=HalfHeightRadius
       width=0
       color {colorUseRule=UseColor_Specified, specifiedColor=<rgba 0 0 0 0>}
       cornerRadii=<0 0 0 0>
      }
      tintColor {colorUseRule=UseColor_Specified, specifiedColor=<rgba 1 1 1 1>}
      uVRegion {bIsValid=false, max=<0 0>, min=<0 0>}
     }
     selectedBackgroundColor {colorUseRule=UseColor_Foreground, specifiedColor=<rgba 1 0 1 1>}
     shadowColorAndOpacity=<rgba 0 0 0 1>
     shadowOffset=<0 0>
     strikeBrush {
      bIsDynamicallyLoaded=false
      drawAs=Image
      imageType=NoImage
      mirroring=NoMirror
      resourceName=None
      resourceObject=null
      tiling=NoTile
      imageSize=<32 32>
      margin=<ltrb 0 0 0 0>
      outlineSettings {
       bUseBrushTransparency=false
       roundingType=HalfHeightRadius
       width=0
       color {colorUseRule=UseColor_Specified, specifiedColor=<rgba 0 0 0 0>}
       cornerRadii=<0 0 0 0>
      }
      tintColor {colorUseRule=UseColor_Specified, specifiedColor=<rgba 1 1 1 1>}
      uVRegion {bIsValid=false, max=<0 0>, min=<0 0>}
     }
     underlineBrush {
      bIsDynamicallyLoaded=false
      drawAs=Image
      imageType=NoImage
      mirroring=NoMirror
      resourceName=None
      resourceObject=null
      tiling=NoTile
      imageSize=<32 32>
      margin=<ltrb 0 0 0 0>
      outlineSettings {
       bUseBrushTransparency=false
       roundingType=HalfHeightRadius
       width=0
       color {colorUseRule=UseColor_Specified, specifiedColor=<rgba 0 0 0 0>}
       cornerRadii=<0 0 0 0>
      }
      tintColor {colorUseRule=UseColor_Specified, specifiedColor=<rgba 1 1 1 1>}
      uVRegion {bIsValid=false, max=<0 0>, min=<0 0>}
     }
    }
    renderTransform {angle=0, scale=<1 1>, shear=<-4 0>, translation=<0 0>}
   }
   slot {class=OverlaySlot, properties {horizontalAlignment=HAlign_Center}}
  }
 ]
}
